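# Animation Blueprint for the Zeus AI's DBNO (knocked-down) poses on the Fortnite player skeleton.
# A conduit-hub state machine routes between injured idle, crawl, turn, fall and rise-transform, with both airborne states landing through DBNO Land.
# When airborne, the update event picks transform over normal fall if the pawn has the StartFloating gameplay tag.
# The output gets large-body spine corrections, then an additive crouch lean.
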AnimBlueprint /AIGods_Zeus/Animation/Zeus_DBNO_AnimBP
target_skeleton=&Fortnite_M_Avg_Player_Skeleton
graphs [
 AnimGraph {nodes [
   n1 = Blendspace Player 'Crouch_Lean'!(
    X=Get FramerateDependentLeanAngle() -> (FramerateDependentLeanAngle:float)
    PlayRate:float=Get SpeedAdjustedPlayrate() -> (SpeedAdjustedPlayrate:float)
   ) -> (Pose) {
    blendSpace=/Game/Animation/Game/MainPlayer/Locomotion/Medium/Male/Crouch/Lean/Crouch_Lean
    groupName=Locomotion
    groupRole=AlwaysFollower
    method=SyncGroup
   }
   AnimGraphNode_StateMachine_194 = DBNO() {is=title AnimGraphNode_StateMachine, bAllowConduitEntryStates=true}
   n2 = Get bIsLargeBodyType() -> (bIsLargeBodyType:bool)
   n3 = "Transform (Modify) Bone - Bone: Spine_02"!(
    ComponentPose=Local To Component(LocalPose=@AnimGraphNode_StateMachine_194)
    bAlphaBoolEnabled:bool=@n2
    Translation="0,0,0"
    Rotation="0,0,-8"
    Scale="1,1,1"
   ) -> (Pose) {alphaInputType=Bool, rotationMode=BMM_Additive, alphaBoolBlend {blendOption=Linear}, boneToModify {boneName=Spine_02}, rotation {roll=-8}}
   n4 = "Transform (Modify) Bone - Bone: spine_05"!(ComponentPose=@n3, bAlphaBoolEnabled:bool=@n2, Translation="0,0,0", Rotation="0,0,4", Scale="1,1,1") -> (Pose) {alphaInputType=Bool, rotationMode=BMM_Additive, alphaBoolBlend {blendOption=Linear}, boneToModify {boneName=spine_05}, rotation {roll=4}}
   Output Pose(Result=Apply Additive(Base=Component To Local(ComponentPose=@n4), Additive=@n1, Alpha=Get LeanAdditiveAlpha() -> (LeanAdditiveAlpha:float)) {alphaBoolBlend {blendOption=Linear}}) {layerGroup=None}
  ]}
 DBNO {
  entry=Conduit
  path=AnimGraph.AnimGraphNode_StateMachine_194.DBNO
  ids=id
  nodes [
   AnimStateConduitNode_0 = Conduit()
   AnimStateNode_0 = DBNO Fall() {is=title AnimStateNode}
   AnimStateNode_1 = DBNO Land() {is=title AnimStateNode}
   AnimStateNode_2 = DBNO Transform() {is=title AnimStateNode}
   AnimStateNode_4 = DBNO Turn() {is=title AnimStateNode}
   AnimStateNode_5 = DBNO Idle() {is=title AnimStateNode}
   AnimStateNode_9 = DBNO Move() {is=title AnimStateNode}
   AnimStateTransitionNode_0 = AnimStateTransitionNode() {logicType=TLT_Inertialization}
   AnimStateTransitionNode_1 = AnimStateTransitionNode()
   AnimStateTransitionNode_14 = AnimStateTransitionNode() {crossfadeDuration=0.3, logicType=TLT_Inertialization, priorityOrder=100}
   AnimStateTransitionNode_16 = AnimStateTransitionNode() {logicType=TLT_Inertialization}
   AnimStateTransitionNode_17 = AnimStateTransitionNode()
   AnimStateTransitionNode_18 = AnimStateTransitionNode() {crossfadeDuration=0.33, logicType=TLT_Inertialization, priorityOrder=2}
   AnimStateTransitionNode_19 = AnimStateTransitionNode()
   AnimStateTransitionNode_20 = AnimStateTransitionNode()
   AnimStateTransitionNode_21 = AnimStateTransitionNode() {crossfadeDuration=0.3, logicType=TLT_Inertialization}
   AnimStateTransitionNode_24 = AnimStateTransitionNode() {bAutomaticRuleBasedOnSequencePlayerInState=true}
   AnimStateTransitionNode_25 = AnimStateTransitionNode()
   AnimStateTransitionNode_3 = AnimStateTransitionNode() {crossfadeDuration=0.067, logicType=TLT_Inertialization}
  ]
  links [
   {
    dst=DBNO Fall
    rule=AnimGraph.AnimGraphNode_StateMachine_194.DBNO.AnimStateTransitionNode_21.AnimationTransitionGraph_0
    src=Conduit
    via=AnimStateTransitionNode_21
   }
   {
    dst=DBNO Idle
    rule=AnimGraph.AnimGraphNode_StateMachine_194.DBNO.AnimStateTransitionNode_14.Transition
    src=Conduit
    via=AnimStateTransitionNode_14
   }
   {
    dst=DBNO Move
    rule=AnimGraph.AnimGraphNode_StateMachine_194.DBNO.AnimStateTransitionNode_16.Transition
    src=Conduit
    via=AnimStateTransitionNode_16
   }
   {
    dst=DBNO Transform
    rule=AnimGraph.AnimGraphNode_StateMachine_194.DBNO.AnimStateTransitionNode_0.Transition
    src=Conduit
    via=AnimStateTransitionNode_0
   }
   {
    dst=DBNO Turn
    rule=AnimGraph.AnimGraphNode_StateMachine_194.DBNO.AnimStateTransitionNode_18.Transition
    src=Conduit
    via=AnimStateTransitionNode_18
   }
   {
    dst=DBNO Land
    rule=AnimGraph.AnimGraphNode_StateMachine_194.DBNO.AnimStateTransitionNode_3.AnimationTransitionGraph_0
    src=DBNO Fall
    via=AnimStateTransitionNode_3
   }
   {
    dst=Conduit
    rule=AnimGraph.AnimGraphNode_StateMachine_194.DBNO.AnimStateTransitionNode_20.Transition
    src=DBNO Idle
    via=AnimStateTransitionNode_20
   }
   {
    dst=Conduit
    rule=AnimGraph.AnimGraphNode_StateMachine_194.DBNO.AnimStateTransitionNode_24.Transition
    src=DBNO Land
    via=AnimStateTransitionNode_24
   }
   {
    dst=Conduit
    rule=AnimGraph.AnimGraphNode_StateMachine_194.DBNO.AnimStateTransitionNode_25.Transition
    src=DBNO Land
    via=AnimStateTransitionNode_25
   }
   {
    dst=Conduit
    rule=AnimGraph.AnimGraphNode_StateMachine_194.DBNO.AnimStateTransitionNode_17.Transition
    src=DBNO Move
    via=AnimStateTransitionNode_17
   }
   {
    dst=DBNO Land
    rule=AnimGraph.AnimGraphNode_StateMachine_194.DBNO.AnimStateTransitionNode_1.Transition
    src=DBNO Transform
    via=AnimStateTransitionNode_1
   }
   {
    dst=Conduit
    rule=AnimGraph.AnimGraphNode_StateMachine_194.DBNO.AnimStateTransitionNode_19.Transition
    src=DBNO Turn
    via=AnimStateTransitionNode_19
   }
  ]
 }
 Conduit {path=AnimGraph.AnimGraphNode_StateMachine_194.DBNO.AnimStateConduitNode_0.Conduit, nodes [Result(bCanEnterTransition=true)]}
 DBNO Fall {
  path=AnimGraph.AnimGraphNode_StateMachine_194.DBNO.AnimStateNode_0.DBNO Fall
  nodes [Output Animation Pose(Result=Sequence Player 'DBNO_Fall_Loop'() {bLoopAnimation=false, sequence=/Game/Animation/Game/MainPlayer/Locomotion/Medium/Male/DBNO/DBNO_Fall_Loop, startPosition=0.25})]
 }
 DBNO Land {
  path=AnimGraph.AnimGraphNode_StateMachine_194.DBNO.AnimStateNode_1.DBNO Land
  nodes [Output Animation Pose(Result=Sequence Player 'DBNO_Fall_Land'() {bLoopAnimation=false, sequence=/Game/Animation/Game/MainPlayer/Locomotion/Medium/Male/DBNO/DBNO_Fall_Land, startPosition=0.067})]
 }
 DBNO Transform {
  path=AnimGraph.AnimGraphNode_StateMachine_194.DBNO.AnimStateNode_2.DBNO Transform
  nodes [Output Animation Pose(Result=Sequence Player 'AI_RiseFromDBNO'() {bLoopAnimation=false, sequence=&/AI_RiseFromDBNO, startPosition=0.25})]
 }
 DBNO Turn {
  path=AnimGraph.AnimGraphNode_StateMachine_194.DBNO.AnimStateNode_4.DBNO Turn
  nodes [Output Animation Pose(Result=Sequence Player 'DBNOCrawl_E_CMM'!(PlayRate:float=Get DBNOTurnPlayRate() -> (DBNOTurnPlayRate:float), StartPosition:float=.36) -> (Pose) {sequence=/Game/Animation/Game/MainPlayer/Locomotion/Medium/Male/DBNO/DBNOCrawl_E_CMM, startPosition=0.5})]
 }
 DBNO Idle {
  path=AnimGraph.AnimGraphNode_StateMachine_194.DBNO.AnimStateNode_5.DBNO Idle
  nodes [Output Animation Pose(Result=Sequence Player 'Zeusinjured'() {sequence=&/Zeusinjured})]
 }
 DBNO Move {
  path=AnimGraph.AnimGraphNode_StateMachine_194.DBNO.AnimStateNode_9.DBNO Move
  nodes [Output Animation Pose(Result=Blendspace Player 'BS_DBNO_CMM'(X=Get LocalVelocityYawAngle() -> (LocalVelocityYawAngle:float), Y=Get MovementBlendSpaceSpeed2D() -> (MovementBlendSpaceSpeed2D:float)) {blendSpace=/Game/Animation/Game/MainPlayer/Locomotion/Medium/Male/DBNO/BS_DBNO_CMM, groupName=Locomotion, method=SyncGroup})]
 }
 Transition {
  path=AnimGraph.AnimGraphNode_StateMachine_194.DBNO.AnimStateTransitionNode_0.Transition
  nodes [Result(bCanEnterTransition=Get bIsTransformationFalling() -> (bIsTransformationFalling:bool))]
 }
 Transition {
  path=AnimGraph.AnimGraphNode_StateMachine_194.DBNO.AnimStateTransitionNode_1.Transition
  nodes [Result(bCanEnterTransition=Get bIsOnGround() -> (bIsOnGround:bool))]
 }
 Transition {path=AnimGraph.AnimGraphNode_StateMachine_194.DBNO.AnimStateTransitionNode_14.Transition, nodes [Result(bCanEnterTransition=true)]}
 Transition {
  path=AnimGraph.AnimGraphNode_StateMachine_194.DBNO.AnimStateTransitionNode_16.Transition
  nodes [Result(bCanEnterTransition=Get bTransition_FullBody_DBNOMove() -> (bTransition_FullBody_DBNOMove:bool))]
 }
 Transition {path=AnimGraph.AnimGraphNode_StateMachine_194.DBNO.AnimStateTransitionNode_17.Transition, nodes [Result(bCanEnterTransition=true)]}
 Transition {
  path=AnimGraph.AnimGraphNode_StateMachine_194.DBNO.AnimStateTransitionNode_18.Transition
  nodes [Result(bCanEnterTransition=Get bTransition_FullBody_DBNOTurn() -> (bTransition_FullBody_DBNOTurn:bool))]
 }
 Transition {
  path=AnimGraph.AnimGraphNode_StateMachine_194.DBNO.AnimStateTransitionNode_19.Transition
  nodes [Result(bCanEnterTransition=NOT Boolean(A=Get bTransition_FullBody_DBNOTurn()))]
 }
 Transition {path=AnimGraph.AnimGraphNode_StateMachine_194.DBNO.AnimStateTransitionNode_20.Transition, nodes [Result(bCanEnterTransition=true)]}
 AnimationTransitionGraph_0 {
  path=AnimGraph.AnimGraphNode_StateMachine_194.DBNO.AnimStateTransitionNode_21.AnimationTransitionGraph_0
  nodes [Result(bCanEnterTransition=Get bIsNormalFalling() -> (bIsNormalFalling:bool))]
 }
 Transition {path=AnimGraph.AnimGraphNode_StateMachine_194.DBNO.AnimStateTransitionNode_24.Transition, nodes [Result()]}
 Transition {
  path=AnimGraph.AnimGraphNode_StateMachine_194.DBNO.AnimStateTransitionNode_25.Transition
  nodes [Result(bCanEnterTransition=NOT Boolean(A=Get bIsOnGround()))]
 }
 AnimationTransitionGraph_0 {
  path=AnimGraph.AnimGraphNode_StateMachine_194.DBNO.AnimStateTransitionNode_3.AnimationTransitionGraph_0
  nodes [Result(bCanEnterTransition=Get bIsOnGround())]
 }
 EventGraph {nodes [
   Event BlueprintUpdateAnimation()
   > n1 = Branch(Condition=Get bIsOnGround())
   > Set bIsNormalFalling(bIsNormalFalling:bool=false) -> (Output_Get:bool)
   > Set bIsTransformationFalling(bIsTransformationFalling:bool=false) -> (Output_Get:bool)
   n2 = Cast To FortPawn!(execute=@n1.else, Object=TryGetPawnOwner()) -> (then, CastFailed, AsFort Pawn:Fort Pawn Object Reference) {is=title K2Node_DynamicCast}
   n3 = GetAbilitySystemComponent!(Actor:Actor Object Reference=@n2) -> (ReturnValue:Ability System Component Object Reference)
   > n4 = Is Valid(InputObject=@n3)
   Set bIsNormalFalling(execute=[@n4.Is Not Valid, @n2.CastFailed], bIsNormalFalling=true)
   > Set bIsTransformationFalling(bIsTransformationFalling=false)
   n5 = Branch(
    execute=@n4.Is Valid
    Condition=HasMatchingGameplayTag!(self:Gameplay Tag Asset Interface Interface=@n3, TagToCheck:Gameplay Tag Structure=(TagName="Event.Ability.AIGods.Zeus.StartFloating")) -> (ReturnValue:bool)
   )
   > Set bIsNormalFalling(bIsNormalFalling=false)
   > Set bIsTransformationFalling(bIsTransformationFalling=true)
   Set bIsTransformationFalling(execute=@n5.else, bIsTransformationFalling=false)
   > Set bIsNormalFalling(bIsNormalFalling=true)
  ]}
]
variables [
 AdjustLargeBody {type=boolean}
 bIsNormalFalling {type=boolean}
 bIsTransformationFalling {type=boolean}
]
