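; Shield actor for the Zeus AI boss, providing shield visuals that react to the boss's floating and enlarging phases.
; On BeginPlay it hides its decal on Medium and Low effects quality and, after a short delay, drives shield opacity by timeline on cosmetic clients.
; On attach it ignores the owner's capsule and awaits gameplay events.
; Floating activates rebirth Niagara and updates opacity; enlarging waits for the event magnitude, then animates opacity.
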(event EventBeginPlay
  (|CallParentFunction)
  (Utilities|FlowControl|Sequence
    (:then_0
      ;[1 1 1] no FX decals on Medium-Low
      (Utilities|EffectsQualityCheck 0
        (:Medium
          _)
        (:Low
          _))
      (Rendering|SetVisibility (Variables|Default|GetDecal) false false))
    (:then_1
      (Utilities|FlowControl|Delay 0.5)
      (Networking|CanExecuteCosmeticEvents
        (:True
          (bind (_direction _sheildopacity) (Timeline|StartFloating PlayFromStart)
            (:Update
              (Rendering|Material|SetScalarParameterValueonMaterials (Variables|Default|GetShieldMesh) "OverallOpacity" _sheildopacity))))))))

(event Collision|EventActorBeginOverlap
  (params (OtherActor :type Actor%20Object%20Reference)))

(event EventTick
  (params (DeltaSeconds :type Float%20%28single-precision%29)))

(event EventAttachShield
  (params (ShieldOwner :type Actor%20Object%20Reference))
  (CallFunction|ClientDisableShieldCollision ShieldOwner)
  (Ability|Async|WaitGameplayEventToActor ShieldOwner (Variables|Default|GetTag_StartFloating) true
    (:then
      (bind (_asynctaskproxy _payload) (Ability|Async|WaitGameplayEventToActor ShieldOwner (Variables|Default|GetTag_StartEnlarging) true)
        (:EventReceived
          (Development|PrintString "Zeus Started Enlarging Process" true true "(R=1.000000,G=0.084804,B=0.000000,A=1.000000)")
          (Utilities|FlowControl|Delay (.EventMagnitude (Utilities|Struct|BreakGameplayEventData _payload)))
          (CallFunction|ClientScaling))))
    (:EventReceived
      (CallFunction|ClientStartFloating))))

(event Custom|ClientDisableShieldCollision
  (params (PawnOwner :type Actor%20Object%20Reference))
  (bind _asfort_pawn (Utilities|Casting|CastToFortPawn PawnOwner))
  (Collision|IgnoreActorwhenMoving (Variables|/Script/Engine.Character|GetCapsuleComponent _asfort_pawn) self true))

(event Custom|ClientScaling
  (bind (_direction _sheildopacity) (Timeline|ScalingSheildChange PlayFromStart)
    (:Update
      (Rendering|Material|SetScalarParameterValueonMaterials (Variables|Default|GetShieldMesh) "OverallOpacity" _sheildopacity))))

(event Custom|ClientStartFloating
  (Components|Activation|SetActive (Variables|Default|GetNS_ZeusGodAI_RebirthFloating) true true)
  (bind (_direction _sheildopacity) (Timeline|StartFloatingSheildChange PlayFromStart)
    (:Finished
      (Rendering|Material|SetScalarParameterValueonMaterials (Variables|Default|GetShieldMesh) "OverallOpacity" _sheildopacity))))

(fn ConstructionScript
  (|CallParentFunction))
