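# Translucent deferred decal material for Zeus's DBNO (down but not out) effect, drawing a lightning-colored, noise-warped glow around the actor's position.
# A radius sphere mask, perturbed by turbulence noise sampled in radial UVs, feeds two opposed threshold ramps whose product forms a bright band tinted by LightningC.
# The band adds to a curve-atlas gradient indexed by the mask and offset by the noise, then the mask's ceiling clips everything to the radius.
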Material /AIGods_Zeus/Effect/MD_Zeus_DBNO_Decal
graph {
 properties {blendMode=BLEND_Translucent, materialDomain=MD_DeferredDecal, pixelDepthOffsetMode=PDOM_Legacy, usedShadingModels=Default Lit}
 nodes [
  n1 = SphereMask(
   A=WorldPosition() {worldPositionShaderOffset=WPT_ExcludeAllShaderOffsets}.XYZ
   B=ActorPositionWS()
   Radius=ScalarParameter() {defaultValue=280, parameterName=Radius}
  ) {hardnessPercent=0}
  n2 = TextureObjectParameter() {
   group=NoiseA
   parameterName=NoiseA_Texture
   samplerType=SAMPLERTYPE_LinearColor
   sortPriority=0
   texture=/Game/Effects/Fort_Effects/Effects/Environment/Desert/Dust/Organize/T_Turbulence_Seamless
  }
  n3 = VectorParameter() {group=NoiseA, parameterName=NoiseA_UVControl, channelNames {a=VPan, b=UPan, g=VScale, r=UScale}, defaultValue=<rgba 1 1 0 -0.7>}
  n4 = VectorParameter() {
   group=NoiseA
   parameterName=NoiseA_NoiseControl
   sortPriority=60
   channelNames {a=VStrength, b=UStrength, g=VOffset, r=UOffset}
   defaultValue=<rgba -0.5 -0.5 1.2 0>
  }
  n5 = MaterialFunctionCall!(
   NoiseValue=MaterialFunctionCall(
    Texture=@n2
    UVControl_RGB=@n3
    UVControl_A=@n3
    "[Override]UVInput"=MaterialFunctionCall() {materialFunction=&VectorToRadialValue}.Radial Coordinates
   ) {materialFunction=&FXCommon_TextureSample}.RG
   Control_RGB=@n4
   Control_A=@n4
  ) {materialFunction=/Game/Effects/Niagara/MasterMaterials/Functions/Common/FXCommon_NoiseControl}
  n6 = Add(A=@n1, B=ComponentMask(None=@n5.NoiseValue) {r=true})
  n7 = ScalarParameter() {group=ScanL, parameterName=ScanL_GrayAdjust_Add}
  n8 = Saturate(None=Multiply(A=Add(A=@n6, B=@n7), B=ScalarParameter() {defaultValue=5, group=ScanL, parameterName=ScanL_GrayAdjust_Mul}))
  n9 = ScalarParameter() {defaultValue=30, group=ScanL, parameterName=ScanL_GrayAdjust_Power}
  n10 = OneMinus(None=@n6)
  n11 = ScalarParameter() {defaultValue=-0.8, group=ScanR, parameterName=ScanR_GrayAdjust_Add}
  n12 = Saturate(None=Multiply(A=Add(A=@n10, B=@n11), B=ScalarParameter() {defaultValue=20, group=ScanR, parameterName=ScanR_GrayAdjust_Mul}))
  n13 = ScalarParameter() {defaultValue=1, group=ScanR, parameterName=ScanR_GrayAdjust_Power}
  n14 = Multiply(A=Saturate(None=Power(Base=@n8, Exp=@n9)), B=Saturate(None=Power(Base=@n12, Exp=@n13)))
  n15 = VectorParameter() {parameterName=LightningC, defaultValue=<rgba 0 16.6667 50 1>}
  n16 = Add(
   A=MaterialFunctionCall!(CurveIndex, CurveNum, GreyMapIn=@n1) {materialFunction=/Game/Effects/Niagara/MasterMaterials/Functions/Common/FXCommon_CurveAtlasUV}.UV
   B=@n5.NoiseValue
  )
  Multiply(
   A=Ceil(None=@n1)
   B=Add(
    A=Multiply(A=@n14, B=@n15.RGB)
    B=TextureSample(UVs=@n16) {samplerType=SAMPLERTYPE_LinearColor, texture=&/CLA_ZeusGodAI}.RGB
   )
  )
 ]
}
