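# Translucent unlit energy shield material for the Zeus AI gods content, with animated hit reactions.
# Emissive layers noise-distorted lightning and plasma with an impact ring of polar-mapped lightning around HitWave-Loc, zeroed on forward and mobile paths.
# Opacity combines flow-modulated Fresnel, view-angle and depth fades, with inverted depth fade boosting intersections.
# World position offset adds panning noise wobble and a hit displacement timed from the GameTime timestamp.
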Material /AIGods_Zeus/Effect/M_EnergyShield
graph {
 properties {blendMode=BLEND_Translucent, pixelDepthOffsetMode=PDOM_Legacy, shadingModel=MSM_Unlit, usedShadingModels=Unlit}
 nodes [
  n1 = NamedRerouteUsage()
  n2 = Power(Base=@n1, Exp=ScalarParameter() {defaultValue=0.5, group=HitWave, parameterName=HitWave-OuterRingDelay}) {constExponent=0.7}
  n3 = Multiply(A=ScalarParameter() {defaultValue=300, group=HitWave, parameterName=HitWave-SizeMultiplierOuterRing}, B=@n2)
  n4 = Divide(A=@n3)
  n5 = VectorParameter() {group=HitWave, parameterName=HitWave-Loc, defaultValue=<rgba 100 -1400 -150 0>}
  n6 = ComponentMask(None=@n5.RGB) {b=true, g=true, r=true}
  n7 = WorldPosition() {worldPositionShaderOffset=WPT_ExcludeAllShaderOffsets}
  n8 = Saturate(None=LinearInterpolate(A=ScalarParameter() {defaultValue=-0.1, group=HitWave, parameterName=HitWave-InnerRingDelay}, Alpha=@n1) {constA=-0.7})
  n9 = Multiply(A=@n8, B=ScalarParameter() {defaultValue=1000, group=HitWave, parameterName=HitWave-SizeMultiplierInnerRing})
  n10 = LinearInterpolate(A=Divide(A=@n9), B=@n9, Alpha=@n8)
  n11 = ScalarParameter() {defaultValue=0.25, group=HitWave, parameterName=Hitwave-SphereHard2}
  n12 = Multiply(
   A=SphereMask(
    A=@n7.XYZ
    B=@n6
    Radius=LinearInterpolate(A=@n4, B=@n3, Alpha=@n2)
    Hardness=ScalarParameter() {defaultValue=0.7, group=HitWave, parameterName=Hitwave-SphereHard1}
   ) {hardnessPercent=0}
   B=OneMinus(None=SphereMask(A=@n7.XYZ, B=@n6, Radius=@n10, Hardness=@n11) {hardnessPercent=0})
  )
  n13 = Constant()
  n14 = NamedRerouteDeclaration(None=ShadingPathSwitch(Default=@n12, Deferred=@n12, Forward=@n13, Mobile=@n13)) {name=ImpactShockwaveInfo, nodeColor=<rgba 0.2187 1 0 1>}
  n15 = TextureObjectParameter() {
   group=LightningCircle
   parameterName=LightningCircle_Texture
   samplerType=SAMPLERTYPE_LinearColor
   sortPriority=0
   texture=/Game/Effects/Fort_Effects/Emotes/UltraLight/Texture/T_Plasma_A_C
  }
  n16 = TextureObjectParameter() {
   group=NoiseA
   parameterName=NoiseA_Texture
   samplerType=SAMPLERTYPE_LinearColor
   sortPriority=0
   texture=/Game/Effects/Fort_Effects/Effects/Environment/Desert/Dust/Organize/T_Turbulence_Seamless
  }
  n17 = VectorParameter() {group=NoiseA, parameterName=NoiseA_UVControl, channelNames {a=VPan, b=UPan, g=VScale, r=UScale}, defaultValue=<rgba 1 0.25 0 0.3>}
  n18 = VectorParameter() {
   group=NoiseA
   parameterName=NoiseA_NoiseControl
   sortPriority=60
   channelNames {a=VStrength, b=UStrength, g=VOffset, r=UOffset}
   defaultValue=<rgba -0.5 -0.5 0 1>
  }
  n19 = MaterialFunctionCall!(
   NoiseValue=MaterialFunctionCall(Texture=@n16, UVControl_RGB=@n17, UVControl_A=@n17, "[Override]TimeInput"=Time() {bOverride_Period=true, period=6666}) {materialFunction=&FXCommon_TextureSample}.RG
   Control_RGB=@n18
   Control_A=@n18
  ) {materialFunction=/Game/Effects/Niagara/MasterMaterials/Functions/Common/FXCommon_NoiseControl}
  n20 = VectorParameter() {
   group=LightningCircle
   parameterName=LightningCircle_UVControl
   channelNames {a=VPan, b=UPan, g=VScale, r=UScale}
   defaultValue=<rgba 1.2 0.5 1 0>
  }
  n21 = ScalarParameter() {defaultValue=300, parameterName=ADW}
  n22 = VectorParameter() {parameterName=HitWave-Normal}
  n23 = Subtract(A=@n7.XYZ, B=@n5.RGB)
  n24 = Normalize(VectorInput=MaterialFunctionCall!(Plane Normal=@n22.RGB, Vector=Normalize(VectorInput=@n23)) {materialFunction=/Engine/Functions/Engine_MaterialFunctions02/Math/ProjectVectorOntoPlane}.Result)
  n25 = Normalize(VectorInput=CrossProduct(A=@n22.RGB, B=Constant3Vector() {constant=<rgba 0 0 1 0>}))
  n26 = MaterialFunctionCall(
   Texture=@n15
   AdditionalUVOffset=@n19.NoiseValue
   UVControl_RGB=@n20
   UVControl_A=@n20
   "[Override]UVInput"=AppendVector(
    A=SphereMask(A=@n5.RGB, B=@n7.XYZ, Radius=@n21) {hardnessPercent=0}
    B=MaterialFunctionCall!(Normalized V1=@n24, Normalized V2=@n25) {materialFunction=/Engine/Functions/Engine_MaterialFunctions02/Math/AngleBetweenVectors}.0-1
   )
  ) {materialFunction=&FXCommon_TextureSample}
  n27 = ScalarParameter() {defaultValue=0.3, parameterName=CCC}
  n28 = VectorParameter() {parameterName=CircleLightningColor, defaultValue=<rgba 8 32 80 3>}
  n29 = Multiply(A=Power(Base=Desaturation(None=@n26.RGB, Fraction=@n27), Exp=@n28.A), B=@n28.RGB)
  n30 = MaterialFunctionCall(Float2=TextureCoordinate()) {materialFunction=&BreakOutFloat2Components}
  n31 = Saturate(None=Power(Base=Multiply(A=@n30.G) {constB=1.3}) {constExponent=1.3})
  n32 = TextureObjectParameter() {
   group=MainFlow
   parameterName=MainFlow_Texture
   sortPriority=0
   texture=/Game/Characters/Player/Female/Medium/Bodies/F_MED_Dark_Viking_01/Textures/Lightning_1_M
  }
  n33 = ScalarParameter() {defaultValue=1, parameterName=TotalNoiseForLightning}
  n34 = TextureObjectParameter() {
   group=NoiseForLightning
   parameterName=NoiseForLightning_Texture
   samplerType=SAMPLERTYPE_LinearGrayscale
   sortPriority=0
   texture=/Game/Effects/Fort_Effects/Emotes/UltraLight/Texture/T_Tiling_Noise_01_GreyScale
  }
  n35 = VectorParameter() {
   group=NoiseForLightning
   parameterName=NoiseForLightning_UVControl
   channelNames {a=VPan, b=UPan, g=VScale, r=UScale}
   defaultValue=<rgba 4 2 0 0.05>
  }
  n36 = MaterialFunctionCall(Texture=@n34, UVControl_RGB=@n35, UVControl_A=@n35) {materialFunction=&FXCommon_TextureSample}
  n37 = VectorParameter() {
   group=NoiseForLightning
   parameterName=NoiseForLightning_NoiseControl
   sortPriority=60
   channelNames {a=VStrength, b=UStrength, g=VOffset, r=UOffset}
   defaultValue=<rgba -0.5 -0.5 0.3 0>
  }
  n38 = VectorParameter() {group=MainFlow, parameterName=MainFlow_UVControl, channelNames {a=VPan, b=UPan, g=VScale, r=UScale}, defaultValue=<rgba 6 3 0.1 0>}
  n39 = MaterialFunctionCall(
   Texture=@n32
   AdditionalUVOffset=Add(
    A=Multiply(A=@n19.NoiseValue, B=@n33)
    B=MaterialFunctionCall(NoiseValue=@n36.RG, Control_RGB=@n37, Control_A=@n37) {materialFunction=&FXCommon_NoiseControl}.NoiseValue
   )
   UVControl_RGB=@n38
   UVControl_A=@n38
   "[Override]UVInput"=MaterialFunctionCall(XY=TextureCoordinate()) {materialFunction=&Swizzle}.YX
  ) {materialFunction=&FXCommon_TextureSample}
  n40 = Length(None=@n39.RGB) {is=class MaterialExpressionLength}
  n41 = Constant3Vector() {constant=<rgba 1 0 0 0>}
  n42 = Constant3Vector() {constant=<rgba 0 1 0 0>}
  n43 = Constant3Vector() {constant=<rgba 0 0 1 0>}
  n44 = MaterialFunctionCall(Value=Time() {bOverride_Period=true, period=6000}, Period=ScalarParameter() {defaultValue=0.8, parameterName=E}, -1 to 1=StaticBool()) {materialFunction=&LinearSine}
  n45 = Desaturation(
   None=Multiply(
    A=MaterialFunctionCall(A=@n41, B=@n42, C=@n43, Alpha=@n44.Linear Sine) {materialFunction=&Lerp_3Color}.Result
    B=@n39.RGB
   )
   Fraction=Constant() {r=1}
  )
  n46 = Multiply(
   A=Power(Base=LinearInterpolate(A=@n40, B=@n45) {constAlpha=0.4}, Exp=ScalarParameter() {defaultValue=2.5, parameterName=a})
   B=ScalarParameter() {defaultValue=30, parameterName=fw}
  )
  n47 = VectorParameter() {parameterName=LightningColor, defaultValue=<rgba 0 23.3333 70 1>}
  n48 = VectorParameter() {parameterName=LightningFres, defaultValue=<rgba 1.5 0 0 1>}
  n49 = Multiply(
   A=Multiply(A=Multiply(A=@n46, B=@n47.RGB), B=Fresnel(ExponentIn=@n48.R, BaseReflectFractionIn=@n48.R) {baseReflectFraction=0, exponent=1})
   B=Power(Base=MaterialFunctionCall() {materialFunction=&BoundingBoxBased_0-1_UVW}.B) {constExponent=1}
  )
  n50 = MaterialFunctionCall!("Xb1,PS4"=@n49, "PC-High&&Epic,XSX,PS5"=@n49, "PC-Perf,Android", Switch, "PC-Low(NotPerf)&&Medium"=@n49) {materialFunction=/Game/Effects/Niagara/MasterMaterials/Functions/Common/PlatformSwitch_3D}
  n51 = TextureObjectParameter() {
   group=MainEnergy
   parameterName=MainEnergy_Texture
   samplerType=SAMPLERTYPE_LinearGrayscale
   sortPriority=0
   texture=/Game/Effects/Fort_Effects/Emotes/UltraLight/Texture/T_PlasmaCoils
  }
  n52 = VectorParameter() {group=MainEnergy, parameterName=MainEnergy_UVControl, channelNames {a=VPan, b=UPan, g=VScale, r=UScale}, defaultValue=<rgba 2 2 0 0.1>}
  n53 = MaterialFunctionCall(Texture=@n51, AdditionalUVOffset=@n19.NoiseValue, UVControl_RGB=@n52, UVControl_A=@n52) {materialFunction=&FXCommon_TextureSample}
  n54 = ScalarParameter() {defaultValue=2, parameterName=Fwfw}
  n55 = Add(
   A=LinearInterpolate(A=Multiply(A=@n31, B=@n50.Result), Alpha=NamedRerouteUsage()) {constB=0}
   B=LinearInterpolate(
    A=VectorParameter() {defaultValue=<rgba 0.05 0.4458 1 1>}.RGB
    B=VectorParameter() {parameterName=Param2, defaultValue=<rgba 1 7.3333 20 1>}.RGB
    Alpha=Power(Base=@n53.R, Exp=@n54)
   )
  )
  Add(
   A=Multiply(
    A=Saturate(None=@n14)
    B=MaterialFunctionCall("Xb1,PS4"=@n29, "PC-High&&Epic,XSX,PS5"=@n29) {materialFunction=&PlatformSwitch_3D}.Result
   )
   B=@n55
  )
  Constant() {r=1}
  n56 = MaterialFunctionCall() {materialFunction=&CameraDirectionVector}
  n57 = Constant3Vector() {constant=<rgba 0 0 1 0>}
  n58 = LinearInterpolate(
   A=ScalarParameter() {defaultValue=400, parameterName=DepthfadeL}
   B=ScalarParameter() {defaultValue=200, parameterName=DepthfadeL2}
   Alpha=Abs(None=DotProduct(A=@n56.Result, B=@n57))
  )
  n59 = Power(Base=OneMinus(None=DepthFade(FadeDistance=@n58)))
  n60 = ScalarParameter() {defaultValue=3, parameterName=DFS}
  Multiply(A=@n55, B=Add(A=Multiply(A=@n59, B=@n60)))
  n61 = NamedRerouteUsage()
  n62 = VectorParameter() {parameterName=OpacityFlowFresnelParameter, defaultValue=<rgba 3 3 0 0.03>}
  n63 = TextureObjectParameter() {
   group=OpacityFlow
   parameterName=OpacityFlow_Texture
   samplerType=SAMPLERTYPE_LinearColor
   sortPriority=0
   texture=/Game/Items/Traps/Textures/T_Plasma_Flames_01_M
  }
  n64 = VectorParameter() {group=OpacityFlow, parameterName=OpacityFlow_UVControl, channelNames {a=VPan, b=UPan, g=VScale, r=UScale}, defaultValue=<rgba 2 0.6 0 0.3>}
  n65 = MaterialFunctionCall(Texture=@n63, AdditionalUVOffset=@n19.NoiseValue, UVControl_RGB=@n64, UVControl_A=@n64) {materialFunction=&FXCommon_TextureSample}
  n66 = LinearInterpolate(A=@n62.R, B=@n62.R, Alpha=@n65.R) {constB=15}
  n67 = LinearInterpolate(A=@n62.B, B=@n62.B, Alpha=@n65.R)
  n68 = VectorParameter() {parameterName=OpacityFlowFresnelFadeViewAngle, defaultValue=<rgba 6 10 100 5>}
  n69 = Multiply(
   A=MaterialFunctionCall("FadeSensitivity (optional)"=LinearInterpolate(A=@n68.R, B=@n68.R, Alpha=@n65.R) {constA=4, constB=15}) {materialFunction=&FadeBasedOnViewAngle}.Result
   B=LinearInterpolate(A=@n68.B, B=@n68.B, Alpha=@n65.R) {constA=20, constB=0}
  )
  n70 = DepthFade(
   Opacity=Multiply(A=Fresnel(ExponentIn=@n66, BaseReflectFractionIn=@n67) {baseReflectFraction=0, exponent=1}, B=Saturate(None=@n69))
   FadeDistance=ScalarParameter() {defaultValue=50, parameterName=Depthfade}
  )
  n71 = ScalarParameter() {parameterName=OverallOpacity}
  n72 = ScalarParameter() {defaultValue=2, parameterName=DFS2}
  Multiply(A=Add(A=Multiply(A=@n61) {constB=0.5}, B=Multiply(A=@n70, B=@n71)) {constB=0.2}, B=Add(A=Multiply(A=@n59, B=@n72)))
  n73 = Subtract(A=Time(), B=ScalarParameter() {defaultValue=1, desc=Time-stamp, group=HitWave, parameterName=GameTime})
  n74 = ScalarParameter() {defaultValue=0.5, group=HitWave, parameterName=HitWave-DurationNew}
  NamedRerouteDeclaration(None=Saturate(None=Divide(A=@n73, B=@n74))) {name="Impactinfo(0-1)", nodeColor=<rgba 0.9922 1 0 1>}
  n75 = Power(Base=Saturate(None=NamedRerouteUsage()), Exp=ScalarParameter() {defaultValue=1, group=HitWave, parameterName=Hitwave-pow})
  n76 = Multiply(A=Multiply(A=Saturate(None=@n75), B=ScalarParameter() {defaultValue=45, group=HitWave, parameterName=WPO-AddWave2}), B=VertexNormalWS())
  n77 = Multiply(
   A=SphereMask(A=WorldPosition().XYZ, B=VectorParameter() {group=HitWave, parameterName=HitWave-Loc, defaultValue=<rgba 100 -1400 -150 0>}.RGB) {hardnessPercent=0}
   B=Constant() {r=5}
  )
  n78 = Subtract(A=Time() {period=6666}, B=ScalarParameter() {defaultValue=1, desc=Time-stamp, group=HitWave, parameterName=GameTime})
  n79 = Sine(None=Saturate(None=Divide(A=@n78) {constB=0.2}))
  n80 = Transform(None=Multiply(A=VectorParameter() {parameterName=HitWave-Normal}.RGB, B=Multiply(A=@n77, B=@n79)) {constB=100}) {transformSourceType=TRANSFORMSOURCE_Local}
  n81 = TextureCoordinate()
  n82 = Sine(None=Multiply(A=MaterialFunctionCall(Float2=@n81) {materialFunction=&BreakOutFloat2Components}.G) {constB=0.5})
  n83 = Constant() {r=1}
  n84 = Constant() {r=1.2}
  n85 = Panner(Coordinate=Multiply(A=@n81, B=AppendVector(A=@n83, B=@n84)), Time=Time() {bOverride_Period=true, period=6666}) {speedX=0.07, speedY=-0.25}
  n86 = Multiply(
   A=VertexNormalWS()
   B=Multiply(
    A=Saturate(None=@n82)
    B=TextureSample(UVs=@n85) {samplerType=SAMPLERTYPE_Grayscale, texture=&LowResBlurredNoise}.RGB
   )
  )
  n87 = Multiply(A=ScalarParameter() {defaultValue=10, parameterName=WPO}, B=DynamicParameter() {paramNames [WPO,Param2,Param3,Param4]}.WPO)
  NamedRerouteDeclaration(None=Add(A=Add(A=@n76, B=@n80) {constB=0}, B=Multiply(A=@n86, B=@n87))) {name=HitImpactWPO, nodeColor=<rgba 0.6641 0 1 1>}
  NamedRerouteUsage()
 ]
}
