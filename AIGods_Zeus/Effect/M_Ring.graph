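# Unlit translucent Niagara sprite material, named as a ring in the Zeus effects, that draws a glowing, animated radial band.
# VectorToRadialValue supplies radial UVs to panning noise, which distorts both a plasma texture and a dust "Sand" texture.
# The refined plasma, an edge mask on one radial axis and particle alpha form the opacity-style chain.
# Powered dust tinted by a very bright SandColor, plus a base color, passes through emissive adjustment and particle color.
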Material /AIGods_Zeus/Effect/M_Ring
graph {
 properties {bUsedWithNiagaraSprites=true, blendMode=BLEND_Translucent, pixelDepthOffsetMode=PDOM_Legacy, shadingModel=MSM_Unlit, usedShadingModels=Unlit}
 nodes [
  Multiply()
  n1 = TextureObjectParameter() {
   group=MainTexture
   parameterName=MainTexture_Texture
   samplerType=SAMPLERTYPE_LinearGrayscale
   sortPriority=0
   texture=/Game/Effects/Fort_Effects/Emotes/UltraLight/Texture/T_Plasma_B_Greyscale
  }
  n2 = TextureObjectParameter() {
   group=NoiseA
   parameterName=NoiseA_Texture
   samplerType=SAMPLERTYPE_LinearGrayscale
   sortPriority=0
   texture=/Game/Effects/Fort_Effects/Emotes/UltraLight/Texture/T_Tiling_Noise_01_GreyScale
  }
  n3 = VectorParameter() {group=NoiseA, parameterName=NoiseA_UVControl, channelNames {a=VPan, b=UPan, g=VScale, r=UScale}, defaultValue=<rgba 1 0.2 0 -0.5>}
  n4 = MaterialFunctionCall() {materialFunction=&VectorToRadialValue}
  n5 = VectorParameter() {
   group=NoiseA
   parameterName=NoiseA_NoiseControl
   sortPriority=60
   channelNames {a=VStrength, b=UStrength, g=VOffset, r=UOffset}
   defaultValue=<rgba -0.5 -0.5 1 0>
  }
  n6 = MaterialFunctionCall!(
   NoiseValue=MaterialFunctionCall(Texture=@n2, UVControl_RGB=@n3, UVControl_A=@n3, "[Override]UVInput"=@n4.Radial Coordinates) {materialFunction=&FXCommon_TextureSample}.RG
   Control_RGB=@n5
   Control_A=@n5
  ) {materialFunction=/Game/Effects/Niagara/MasterMaterials/Functions/Common/FXCommon_NoiseControl}
  n7 = VectorParameter() {group=MainTexture, parameterName=MainTexture_UVControl, channelNames {a=VPan, b=UPan, g=VScale, r=UScale}, defaultValue=<rgba 1 0.5 0 -0.15>}
  n8 = MaterialFunctionCall(Texture=@n1, AdditionalUVOffset=@n6.NoiseValue, UVControl_RGB=@n7, UVControl_A=@n7, "[Override]UVInput"=@n4.Radial Coordinates) {materialFunction=&FXCommon_TextureSample}
  n9 = VectorParameter() {parameterName=Refine, defaultValue=<rgba 6 10 0.2 0.5>}
  n10 = MaterialFunctionCall(Float2=@n4.Radial Coordinates) {materialFunction=&BreakOutFloat2Components}
  n11 = VectorParameter() {group=mask, parameterName=mask_EdgeMask_InvertControl, channelNames {a="", b=Power, g=Multiply, r=Offset}, defaultValue=<rgba 0 2 3 1>}
  n12 = VectorParameter() {group=mask, parameterName=mask_EdgeMask_Control, channelNames {a="", b=Power, g=Multiply, r=Offset}, defaultValue=<rgba 0 2.5 15 1>}
  n13 = ParticleColor()
  Multiply(
   A=Multiply(
    A=MaterialFunctionCall(Texture=@n8.RGB, RGB=@n9.RGB, A=@n9.RGB) {materialFunction=&MF_FX_Refine}.Result
    B=MaterialFunctionCall!(UorV=@n10.G, Invert_EdgeControl=@n11, NoInvert_EdgeControl=@n12) {materialFunction=/Game/Effects/Niagara/MasterMaterials/Functions/Common/FXCommon_EdgeMask_Ex}.Mask
   )
   B=@n13.A
  )
  n14 = VectorParameter() {group=Sand, parameterName=Sand_UVControl, channelNames {a=VPan, b=UPan, g=VScale, r=UScale}, defaultValue=<rgba 1 0.5 0 -0.3>}
  n15 = MaterialFunctionCall(
   Texture=TextureObjectParameter() {group=Sand, parameterName=Sand_Texture, sortPriority=0, texture=/Game/Global/Textures/Default/PointSplatter/T_Tiling_Dust}
   AdditionalUVOffset=@n6.NoiseValue
   UVControl_RGB=@n14
   UVControl_A=@n14
   "[Override]UVInput"=@n4.Radial Coordinates
  ) {materialFunction=&FXCommon_TextureSample}
  n16 = ScalarParameter() {defaultValue=2, parameterName=Sandp}
  n17 = Add(
   A=Multiply(A=Power(Base=@n15.R, Exp=@n16), B=VectorParameter() {parameterName=SandColor, defaultValue=<rgba 0 1666.6663 5000 1>}.RGB)
   B=VectorParameter() {parameterName=a, defaultValue=<rgba 0 3.3333 10 1>}.RGB
  )
  Multiply(
   A=MaterialFunctionCall(Emissive=Divide(A=@n17)) {materialFunction=&MF_FX_EmissiveAdjustments}.Result
   B=@n13.RGB
  )
  ParticleColor()
  SphereMask(A=TextureCoordinate(), B=Constant2Vector() {g=0.5, r=0.5}) {attenuationRadius=0.5, hardnessPercent=0}
 ]
}
