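# Niagara system for Zeus character lightning: short-lived, flickering blue arcs between bones or sockets of User.TargetSM.
# Its sole emitter, LeftArm, spawns ribbon points in four groups at a platform-scaled rate.
# Each update samples two bone positions, derives a control point, and bends points along a Bezier curve via scratch modules.
# Ribbons draw arcs, sprites show only at endpoints, and bone distance and system scale drive material parameters.
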NiagaraSystem /AIGods_Zeus/Effect/NS_CharacterLignting
dependencies {
 usedDataInterfaces [&NiagaraDataInterfacePlatformSet, &NiagaraDataInterfaceCurve, &NiagaraDataInterfaceDebugDraw]
 usedDynamicInputs []
 usedModules [&SystemState, &self:ScratchModule]
}
emitters {LeftArm {
  emitterUpdateScript [
   &EmitterState(Inactive Response=::NewEnumerator1, Life Cycle Mode=::NewEnumerator1, Loop Duration=1, UseLoopDelay=false)
   &self:SpawnRibbonRate(
    ParticlesPerRibbon=3
    Rate:NiagaraFloat=&FN_PickFloatDependOnPlatform(
     TargetPlatform=DataInterface() {propertyValues {Platforms {qualityLevelMask=-29}}}
     TargetPlatformActive=&FN_PickFloatDependOnPlatform(
      TargetPlatform=DataInterface() {propertyValues {Platforms {qualityLevelMask=-31}}}
      TargetPlatformActive=&Multiply_Float(A=$User.RibbonSpawnRate, B=2)
      TargetPlatformNoActive=&Multiply_Float(A=$User.RibbonSpawnRate, B=1.5)
     )
     TargetPlatformNoActive=&Multiply_Float(A=$User.RibbonSpawnRate, B=1)
    )
   ) {module=SpawnRibbonRate}
   &SpawnPerFrame(Spawn Count:NiagaraInt32=$Emitter.SpawnInt, Use Spawn Probability=false)
   &SpawnPerFrame(Spawn Count:NiagaraInt32=$Emitter.SpawnInt, Spawn Group=1, Use Spawn Probability=false) {module=SpawnPerFrame001}
   &SpawnPerFrame(Spawn Count:NiagaraInt32=$Emitter.SpawnInt, Spawn Group=2, Use Spawn Probability=false) {module=SpawnPerFrame002}
   &SpawnPerFrame(Spawn Count:NiagaraInt32=$Emitter.SpawnInt, Spawn Group=3, Use Spawn Probability=false) {module=SpawnPerFrame003}
  ]
  particleSpawnScript [
   &self:DifferentGroup(
    Max=5
    Max1=11
    Max2=15
    Max3=20
    Min1=6
    Min2=12
    Min3=16
    RPG:NiagaraInt32=$Emitter.ParticlesPerRibbon
    Simple Counter:NiagaraDataInterfaceSimpleCounter=$Emitter.Simple Counter
   ) {module=DifferentGroup}
   &InitializeParticle(
    Lifetime=/Game/Effects/Niagara/DynamicInputs/Random/FN_SeededRandomRangeFloat(Minimum:NiagaraFloat=0.25, Maximum:NiagaraFloat=0.3, Random Seed:NiagaraInt32=$Particles.Group)
    Ribbon Width=0
    Ribbon Width Mode=::NewEnumerator1
    Sprite Size Mode=::NewEnumerator3
    Uniform Sprite Size=&RandomRangeFloat(Minimum=50, Maximum=100)
    UsePositionOffset=false
    Write Material Random=true
   )
   &CameraOffset(Camera Offset Amount=5)
  ]
  particleUpdateScript [
   &ParticleState()
   &ScaleSpriteSize(
    Scale Sprite Size Mode=::NewEnumerator0
    Uniform Scale Factor=&MakeCustomFloatFromBool(
     Bool=&Boolean_OrOperation(
      Bool1=&SetBoolByFloatComparison(A=$Particles.NormalizeP, B=0, Comparison Type=::NewEnumerator4)
      Bool2=&SetBoolByFloatComparison(A=$Particles.NormalizeP, B=1, Comparison Type=::NewEnumerator4)
     )
     True Float=&MakeCustomFloatFromBool(
      Bool=&SetBoolByFloatComparison(A=&MakeFloatFromVector(Vector=$System.Scale, Channel=::NewEnumerator2), B=1.5)
      True Float=2
      False Float=1
     )
     False Float=0
    )
   )
   &SkeletalMeshLocation(
    Bone / Socket Index=$Particles.Random1
    Bone Sampling Mode=::NewEnumerator4
    MaskSpawnGroup=false
    Orientation Sampling=::NewEnumerator1
    Position Sampling=::NewEnumerator2
    Sampled Position Offset=&RandomRangeVector(Minimum=<-2 -2 -5>, Maximum=<2 2 5>)
    Skeletal Mesh=$User.TargetSM
    Use Sampled Mesh Attribute as Ribbon ID=false
   )
   &SkeletalMeshLocation(
    Bone / Socket Index=$Particles.Random2
    Bone Sampling Mode=::NewEnumerator4
    MaskSpawnGroup=false
    Orientation Sampling=::NewEnumerator1
    Position Sampling=::NewEnumerator2
    Sampled Position Offset=&RandomRangeVector(Minimum=<-2 -2 -5>, Maximum=<2 2 5>)
    Skeletal Mesh=$User.TargetSM
    Use Sampled Mesh Attribute as Ribbon ID=false
   ) {module=SkeletalMeshLocation001}
   /SetVariables(Particles.Position=&LerpPosition(A=$Particles.SkeletalMeshLocation.SampledPosition, B=$Particles.SkeletalMeshLocation001.SampledPosition, Alpha=$Particles.NormalizeP)) {isSetParameters=true}
   &self:GetThirdPoint(
    BoneO1:Quat4f=$Particles.SkeletalMeshLocation.SampledRotation
    BoneO2:Quat4f=$Particles.SkeletalMeshLocation001.SampledRotation
    P1:NiagaraPosition=$Output.SkeletalMeshLocation.SampledPositionWithoutOffset
    P2:NiagaraPosition=$Output.SkeletalMeshLocation001.SampledPositionWithoutOffset
    Rotate:NiagaraFloat=&FN_SeededRandomRangeFloat(Minimum:NiagaraFloat=0, Maximum:NiagaraFloat=360, Random Seed:NiagaraInt32=$Particles.Group)
    Se:NiagaraFloat=&MakeCustomFloatFromBool(
     Bool=&SetBoolByFloatComparison(
      A=&FN_SeededRandomRangeFloat(Minimum:NiagaraFloat=0, Maximum:NiagaraFloat=1, Random Seed:NiagaraInt32=$Particles.Group)
      B=0.5
     )
     True Float=0.3
     False Float=-0.3
    )
    float001:NiagaraFloat=&FloatFromCurve(
     FloatCurve=DataInterface() {propertyValues {Curve {keys [<at 0 1>,<at 1 0.8>]}}}
     Scale Curve=&MakeCustomFloatFromBool(
      Bool=&SetBoolByFloatComparison(A=&MakeFloatFromVector(Vector=$System.Scale, Channel=::NewEnumerator2), B=1.5)
      True Float=0.85
      False Float=1
     )
    )
   ) {module=GetThirdPoint}
   &self:BezCurve(
    NormalizedExeindex:NiagaraFloat=$Particles.NormalizeP
    P1:NiagaraPosition=$Particles.SkeletalMeshLocation.SampledPosition
    P2:NiagaraPosition=$Particles.BezCThirdPoint
    P3:NiagaraPosition=$Particles.SkeletalMeshLocation001.SampledPosition
   ) {module=BezCurve}
   &DynamicMaterialParameters(
    Index 0 Param 1=1
    Index 0 Param 3=&FN_SeededRandomRangeFloat(Minimum:NiagaraFloat=75, Maximum:NiagaraFloat=125, Random Seed:NiagaraInt32=$Particles.Group)
    Index 1 Param 3=&FloatFromCurve(FloatCurve=DataInterface() {propertyValues {Curve {keys [<at 0.5 -0.5>,<at 1 0.3>]}}})
    Index 1 Param 4=10
    Index 2 Param 1=&FloatFromCurve(
     FloatCurve=DataInterface() {propertyValues {Curve {keys [
         <at 0 5 RCIM_Cubic>
         {arriveTangent=-0.061, interpMode=RCIM_Cubic, leaveTangent=-0.061, tangentMode=RCTM_User, time=63.8, value=3}
         <at 63.9 0 RCIM_Constant>
        ]}}}
     CurveIndex=&DistanceBetweenPositions(A=$Particles.SkeletalMeshLocation.SampledPosition, B=$Particles.SkeletalMeshLocation001.SampledPosition)
     Scale Curve=&MakeCustomFloatFromBool(
      Bool=&SetBoolByFloatComparison(A=&MakeFloatFromVector(Vector=$System.Scale, Channel=::NewEnumerator2), B=1.5)
      True Float=1.5
      False Float=1
     )
    )
    Index 3 Param 3=&FloatFromCurve(FloatCurve=DataInterface() {propertyValues {Curve {keys [<at 0 0>,<at 0.8 1>]}}})
    Index 3 Param 4=-1
    Module.Index0Param0Write=true
    Module.Index0Param1Write=true
    Module.Index0Param2Write=true
    Module.Index0Param3Write=true
    Module.Index1Param0Write=true
    Module.Index1Param1Write=true
    Module.Index1Param2Write=true
    Module.Index1Param3Write=true
    Module.Index2Param0Write=true
    Module.Index2Param1Write=true
    Module.Index2Param2Write=true
    Module.Index2Param3Write=true
    Module.Index3Param0Write=true
    Module.Index3Param1Write=true
    Module.Index3Param2Write=true
    Module.Index3Param3Write=true
    Write Parameter Index 1=true
    Write Parameter Index 2=true
    Write Parameter Index 3=true
   )
   &ScaleColor(ScaleA=true, ScaleRGB=true)
   &ScaleColor(
    Scale Alpha=&RandomRangeFloat(Minimum=0, Maximum=1, Evaluation Type=::NewEnumerator1)
    Scale RGB=&VectorFromFloat(Value=&RandomRangeFloat(Minimum=1, Maximum=2, Evaluation Type=::NewEnumerator1))
    ScaleA=true
    ScaleRGB=true
   ) {module=ScaleColor001}
   /SetVariables(Particles.RColor:LinearColor=&MultiplyLinearColor(
     Linear Color=$Particles.Color
     Scale Factor=&MakeFloat4FromVectorAndFloat("Vector (RGB)"=&VectorFromFloat(Value=$User.ScaleRibbonBrightness), "Float (Alpha)"=1)
    )) {isSetParameters=true}
   &ScaleColor(
    Scale Alpha=&Multiply_Float(
     A=&MakeCustomFloatFromBool(
      Bool=&SetBoolByFloatComparison(A=$Particles.NormalizeP, B=0.5)
      True Float=&FloatFromCurve(FloatCurve=DataInterface() {propertyValues {Curve {keys [<at 0.4 0>,<at 0.6 1>]}}})
      False Float=1
     )
     B=&FloatFromCurve(
      FloatCurve=DataInterface() {propertyValues {Curve {keys [
          {arriveTangent=7.595, interpMode=RCIM_Cubic, leaveTangent=7.595, tangentMode=RCTM_User, time=0, value=0}
          {interpMode=RCIM_Cubic, tangentMode=RCTM_User, time=0.3, value=1}
          {arriveTangent=0.1069, tangentMode=RCTM_User, time=0.6, value=0}
         ]}}}
      Scale Curve=0.75
     )
    )
    Scale RGB=&MakeVectorFromLinearColorRGB(Color=&MultiplyLinearColor(
      Linear Color=<rgba 0.07 0.38 1 1>
      Scale Factor=&MakeVector4_FromFloat(Float=&RandomRangeFloat(Minimum=0, Maximum=2))
     ))
    ScaleA=true
    ScaleRGB=true
   ) {module=ScaleColor002}
   &ScaleRibbonWidth(Ribbon Width Scale=&Power(a=&MakeFloatFromVector(Vector=$System.Scale, Channel=::NewEnumerator2), b=15)) {enabled=false}
  ]
  rendererClasses [&NiagaraSpriteRendererProperties, &NiagaraRibbonRendererProperties]
  renderers [
   {rendererClass=&NiagaraSpriteRendererProperties}
   {rendererClass=&NiagaraRibbonRendererProperties, rendererIndex=1}
  ]
 }}
summary {
 systemName=NS_CharacterLignting
 emitters [{emitterName=LeftArm, rendererClasses [&NiagaraSpriteRendererProperties, &NiagaraRibbonRendererProperties]}]
 userVariables [
  User.RibbonSpawnRate {type=NiagaraFloat, defaultValue=0.25, description=""}
  User.ScaleGlowBrightness {type=NiagaraFloat, defaultValue=3, description=""}
  User.ScaleRibbonBrightness {type=NiagaraFloat, defaultValue=3, description=""}
  User.TargetSM {
   type=NiagaraDataInterfaceSkeletalMesh
   defaultValue {
    struct=NiagaraExt_VariableValue_DataInterface
    value {dataInterface=&self:NiagaraDataInterfaceSkeletalMesh_0}
   }
  }
 ]
}
