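# Ambient lightning Niagara effect for the Zeus AI god, attached to the skeletal mesh in User.TargetSM.
# Four near-identical LeftArm emitters, differing in DifferentGroup index range, spawn short-lived ribbon arcs between two sampled bones.
# Custom modules bend each arc along a Bezier curve with a seeded third point; sprites mark endpoints, and color flickers blue.
# Two Empty emitters add fading meshes on six fixed bones, plus a glow sprite and flickering light.
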NiagaraSystem /AIGods_Zeus/Effect/NS_ZeusGodAI_LightningAmbientFX
dependencies {
 usedDataInterfaces [
  /Script/Niagara.NiagaraDataInterfaceSimpleCounter
  &NiagaraDataInterfaceCurve
  &NiagaraDataInterfaceDebugDraw
  &NiagaraDataInterfaceArrayFloat3
  &NiagaraDataInterfaceArrayInt32
 ]
 usedDynamicInputs []
 usedModules [&SystemState]
}
emitters {
 Empty {
  emitterUpdateScript [
   &EmitterState(Inactive Response=::NewEnumerator1, Life Cycle Mode=::NewEnumerator1, Loop Behavior=::NewEnumerator1, Loop Duration=4, UseLoopDelay=false)
   &SpawnBurst_Instantaneous(Spawn Count=6, Use Spawn Probability=false, UseLoopCountLimit=false)
  ]
  particleSpawnScript [&InitializeParticle(
    Color Mode=::NewEnumerator0
    Lifetime=1.5
    Mesh Scale Mode=::NewEnumerator3
    Sprite Size=<118.0405 198.6727>
    Sprite Size Mode=::NewEnumerator4
    UsePositionOffset=false
    Write Material Random=true
   )]
  particleUpdateScript [
   &ParticleState()
   &ScaleMeshSize(Scale Factor=<2 2 2>) {module=ScaleMeshSize001}
   &ScaleMeshSize(Scale Factor=&SelectVectorFromArray(
     Direct Array Index=&ReturnExecIndex()
     Vector Selection Array=DataInterface() {propertyValues {FloatData [<1 0.7 1>, <0.8 0.5 0.3>, <0.8 0.5 0.3>, <0.5 0.5 0.5>, <1 0.4 0.3>, <1 0.4 0.3>]}}
    ))
   &SkeletalMeshLocation(
    Bone / Socket Index=&SelectIntFromArray(
     Direct Array Index=&ReturnExecIndex()
     Int Selection Array=DataInterface() {propertyValues {IntData [20,1,7,21,14,18]}}
    )
    Bone Sampling Mode=::NewEnumerator4
    MaskSpawnGroup=false
    Skeletal Mesh=$User.TargetSM
   )
   &CameraOffset(Camera Offset Amount=60)
   &ScaleColor(
    Scale Alpha=&FloatFromCurve(
     FloatCurve=DataInterface() {propertyValues {Curve {keys [
         {arriveTangent=7.3507, interpMode=RCIM_Cubic, leaveTangent=7.3507, tangentMode=RCTM_User, time=0, value=0}
         <at 0.1 1>
         {arriveTangent=-0.0309, interpMode=RCIM_Cubic, leaveTangent=-0.0309, tangentMode=RCTM_User, time=0.5, value=1}
         {tangentMode=RCTM_User, time=1, value=0}
        ]}}}
     Scale Curve=&RandomRangeFloat(Minimum=0.2, Maximum=0.4)
    )
    ScaleA=true
    ScaleRGB=true
   )
  ]
  rendererClasses [&NiagaraMeshRendererProperties]
  renderers [{rendererClass=&NiagaraMeshRendererProperties, rendererIndex=0}]
 }
 Empty001 {
  emitterUpdateScript [
   &EmitterState(Inactive Response=::NewEnumerator1, Life Cycle Mode=::NewEnumerator1, Loop Behavior=::NewEnumerator1, Loop Duration=4, UseLoopDelay=false)
   &SpawnBurst_Instantaneous(Use Spawn Probability=false, UseLoopCountLimit=false)
  ]
  particleSpawnScript [&InitializeParticle(
    Color=<rgba 0.03 0.12 0.3 1>
    Lifetime=2
    Sprite Size Mode=::NewEnumerator3
    Uniform Sprite Size=201.705
    UsePositionOffset=false
    Write Material Random=true
   )]
  particleUpdateScript [
   &ParticleState()
   &SkeletalMeshLocation(Bone / Socket Index=20, Bone Sampling Mode=::NewEnumerator4, MaskSpawnGroup=false, Skeletal Mesh=$User.TargetSM)
   &ScaleColor(
    Scale Alpha=&FloatFromCurve(FloatCurve=DataInterface() {propertyValues {Curve {keys [
         {arriveTangent=6.7479, interpMode=RCIM_Cubic, leaveTangent=6.7479, tangentMode=RCTM_User, time=0, value=0}
         {interpMode=RCIM_Cubic, tangentMode=RCTM_User, time=0.5, value=1}
         {tangentMode=RCTM_User, time=1, value=0}
        ]}}})
    ScaleA=true
    ScaleRGB=true
   )
   &FN_Light_Attributes(
    Color=<rgba 0 0.6667 2 1>
    Create Color Attribute=true
    Create Position Attribute=true
    Create Radius Attribute=true
    Position=<0 50 50>
    Radius=900
    Scale Color=&FloatFromCurve(
     FloatCurve=DataInterface() {propertyValues {Curve {keys [
         {arriveTangent=8.5788, interpMode=RCIM_Cubic, leaveTangent=8.5788, tangentMode=RCTM_User, time=0, value=0}
         {interpMode=RCIM_Cubic, tangentMode=RCTM_User, time=0.5, value=1}
         {tangentMode=RCTM_User, time=1, value=0}
        ]}}}
     Scale Curve=&RandomRangeFloat(Minimum=30000, Maximum=100000, Evaluation Type=::NewEnumerator1)
    )
   )
  ]
  rendererClasses [&NiagaraSpriteRendererProperties, &NiagaraLightRendererProperties]
  renderers [
   {rendererClass=&NiagaraSpriteRendererProperties}
   {rendererClass=&NiagaraLightRendererProperties, rendererIndex=1}
  ]
 }
 LeftArm {
  emitterUpdateScript [
   &EmitterState(Inactive Response=::NewEnumerator1, Life Cycle Mode=::NewEnumerator1, Loop Behavior=::NewEnumerator1, Loop Duration=4, UseLoopDelay=false)
   /SetVariables() {isSetParameters=true, module=SetVariables_#1}
   &self:SpawnRibbonRate(ParticlesPerRibbon=3, Rate=0.1) {module=SpawnRibbonRate}
   &SpawnPerFrame(Spawn Count:NiagaraInt32=$Emitter.SpawnInt, Use Spawn Probability=false)
  ]
  particleSpawnScript [
   &self:DifferentGroup(Max=5, RPG:NiagaraInt32=$Emitter.ParticlesPerRibbon, Simple Counter:NiagaraDataInterfaceSimpleCounter=$Emitter.Simple Counter) {module=DifferentGroup}
   &InitializeParticle(
    Color=<rgba 2 2 2 1>
    Lifetime=/Game/Effects/Niagara/DynamicInputs/Random/FN_SeededRandomRangeFloat(Minimum:NiagaraFloat=0.25, Maximum:NiagaraFloat=0.3, Random Seed:NiagaraInt32=$Particles.Group)
    Ribbon Width=0
    Ribbon Width Mode=::NewEnumerator1
    Sprite Size Mode=::NewEnumerator3
    Uniform Sprite Size=&RandomRangeFloat(Minimum=100, Maximum=150)
    UsePositionOffset=false
    Write Material Random=true
   )
   &CameraOffset(Camera Offset Amount=5)
  ]
  particleUpdateScript [
   &ParticleState()
   &ScaleSpriteSize(
    Scale Sprite Size Mode=::NewEnumerator0
    Uniform Scale Factor=&MakeCustomFloatFromBool(
     Bool=&Boolean_OrOperation(
      Bool1=&SetBoolByFloatComparison(A=$Particles.NormalizeP, B=0, Comparison Type=::NewEnumerator4)
      Bool2=&SetBoolByFloatComparison(A=$Particles.NormalizeP, B=1, Comparison Type=::NewEnumerator4)
     )
     True Float=1
     False Float=0
    )
   )
   &SkeletalMeshLocation(
    Bone / Socket Index=$Particles.Random1
    Bone Sampling Mode=::NewEnumerator4
    MaskSpawnGroup=false
    Orientation Sampling=::NewEnumerator1
    Position Sampling=::NewEnumerator2
    Sampled Position Offset=&RandomRangeVector(Minimum=<-2 -2 -5>, Maximum=<2 2 5>)
    Skeletal Mesh=$User.TargetSM
    Use Sampled Mesh Attribute as Ribbon ID=false
   )
   &SkeletalMeshLocation(
    Bone / Socket Index=$Particles.Random2
    Bone Sampling Mode=::NewEnumerator4
    MaskSpawnGroup=false
    Orientation Sampling=::NewEnumerator1
    Position Sampling=::NewEnumerator2
    Sampled Position Offset=&RandomRangeVector(Minimum=<-2 -2 -5>, Maximum=<2 2 5>)
    Skeletal Mesh=$User.TargetSM
    Use Sampled Mesh Attribute as Ribbon ID=false
   ) {module=SkeletalMeshLocation001}
   /SetVariables(Particles.Position=&LerpPosition(A=$Particles.SkeletalMeshLocation.SampledPosition, B=$Particles.SkeletalMeshLocation001.SampledPosition, Alpha=$Particles.NormalizeP)) {isSetParameters=true, module=SetVariables_#2}
   &self:GetThirdPoint(
    BoneO1:Quat4f=$Particles.SkeletalMeshLocation.SampledRotation
    BoneO2:Quat4f=$Particles.SkeletalMeshLocation001.SampledRotation
    P1:NiagaraPosition=$Output.SkeletalMeshLocation.SampledPositionWithoutOffset
    P2:NiagaraPosition=$Output.SkeletalMeshLocation001.SampledPositionWithoutOffset
    Rotate:NiagaraFloat=&FN_SeededRandomRangeFloat(Minimum:NiagaraFloat=0, Maximum:NiagaraFloat=360, Random Seed:NiagaraInt32=$Particles.Group)
    Se:NiagaraFloat=&MakeCustomFloatFromBool(
     Bool=&SetBoolByFloatComparison(
      A=&FN_SeededRandomRangeFloat(Minimum:NiagaraFloat=0, Maximum:NiagaraFloat=1, Random Seed:NiagaraInt32=$Particles.Group)
      B=0.5
     )
     True Float=0.3
     False Float=-0.3
    )
    float001:NiagaraFloat=&FloatFromCurve(FloatCurve=DataInterface() {propertyValues {Curve {keys [<at 0 1.2>,<at 1 0.8>]}}})
   ) {module=GetThirdPoint}
   &self:BezCurve(
    NormalizedExeindex:NiagaraFloat=$Particles.NormalizeP
    P1:NiagaraPosition=$Particles.SkeletalMeshLocation.SampledPosition
    P2:NiagaraPosition=$Particles.BezCThirdPoint
    P3:NiagaraPosition=$Particles.SkeletalMeshLocation001.SampledPosition
   ) {module=BezCurve}
   &DynamicMaterialParameters(
    Index 0 Param 1=1
    Index 0 Param 3=&FN_SeededRandomRangeFloat(Minimum:NiagaraFloat=75, Maximum:NiagaraFloat=125, Random Seed:NiagaraInt32=$Particles.Group)
    Index 1 Param 3=&FloatFromCurve(FloatCurve=DataInterface() {propertyValues {Curve {keys [<at 0.5 -0.5>,<at 1 0.3>]}}})
    Index 1 Param 4=10
    Index 2 Param 1=&FloatFromCurve(
     FloatCurve=DataInterface() {propertyValues {Curve {keys [
         <at 0 5 RCIM_Cubic>
         {arriveTangent=-0.061, interpMode=RCIM_Cubic, leaveTangent=-0.061, tangentMode=RCTM_User, time=63.8, value=3}
         <at 63.9 0 RCIM_Constant>
        ]}}}
     CurveIndex=&DistanceBetweenPositions(A=$Particles.SkeletalMeshLocation.SampledPosition, B=$Particles.SkeletalMeshLocation001.SampledPosition)
     Scale Curve=2
    )
    Index 3 Param 3=&FloatFromCurve(FloatCurve=DataInterface() {propertyValues {Curve {keys [<at 0 0>,<at 0.8 1>]}}})
    Index 3 Param 4=-1
    Module.Index0Param0Write=true
    Module.Index0Param1Write=true
    Module.Index0Param2Write=true
    Module.Index0Param3Write=true
    Module.Index1Param0Write=true
    Module.Index1Param1Write=true
    Module.Index1Param2Write=true
    Module.Index1Param3Write=true
    Module.Index2Param0Write=true
    Module.Index2Param1Write=true
    Module.Index2Param2Write=true
    Module.Index2Param3Write=true
    Module.Index3Param0Write=true
    Module.Index3Param1Write=true
    Module.Index3Param2Write=true
    Module.Index3Param3Write=true
    Write Parameter Index 1=true
    Write Parameter Index 2=true
    Write Parameter Index 3=true
   )
   &ScaleColor(ScaleA=true, ScaleRGB=true)
   &ScaleColor(
    Scale Alpha=&RandomRangeFloat(Minimum=-0.5, Maximum=1, Evaluation Type=::NewEnumerator1)
    Scale RGB=&VectorFromFloat(Value=&RandomRangeFloat(Minimum=1, Maximum=2, Evaluation Type=::NewEnumerator1))
    ScaleA=true
    ScaleRGB=true
   ) {module=ScaleColor001}
   /SetVariables(Particles.RColor:LinearColor=$Particles.Color) {isSetParameters=true, module=SetVariables_#3}
   &ScaleColor(
    Scale Alpha=&Multiply_Float(
     A=&MakeCustomFloatFromBool(
      Bool=&SetBoolByFloatComparison(A=$Particles.NormalizeP, B=0.5)
      True Float=&FloatFromCurve(FloatCurve=DataInterface() {propertyValues {Curve {keys [<at 0.4 0>,<at 0.6 1>]}}})
      False Float=1
     )
     B=&FloatFromCurve(FloatCurve=DataInterface() {propertyValues {Curve {keys [
          {arriveTangent=7.595, interpMode=RCIM_Cubic, leaveTangent=7.595, tangentMode=RCTM_User, time=0, value=0}
          {interpMode=RCIM_Cubic, tangentMode=RCTM_User, time=0.3, value=1}
          {arriveTangent=0.1069, tangentMode=RCTM_User, time=0.6, value=0}
         ]}}})
    )
    Scale RGB=&MakeVectorFromLinearColorRGB(Color=&MultiplyLinearColor(
      Linear Color=<rgba 0 0.3337 1 1>
      Scale Factor=&MakeVector4_FromFloat(Float=&RandomRangeFloat(Minimum=0, Maximum=2))
     ))
    ScaleA=true
    ScaleRGB=true
   ) {module=ScaleColor002}
  ]
  rendererClasses [&NiagaraSpriteRendererProperties, &NiagaraRibbonRendererProperties]
  renderers [
   {rendererClass=&NiagaraSpriteRendererProperties}
   {rendererClass=&NiagaraRibbonRendererProperties, rendererIndex=1}
  ]
 }
 LeftArm001 {
  emitterUpdateScript [
   &EmitterState(Inactive Response=::NewEnumerator1, Life Cycle Mode=::NewEnumerator1, Loop Behavior=::NewEnumerator1, Loop Duration=4, UseLoopDelay=false)
   /SetVariables() {isSetParameters=true, module=SetVariables_#1}
   &self:SpawnRibbonRate(ParticlesPerRibbon=3, Rate=0.1) {module=SpawnRibbonRate}
   &SpawnPerFrame(Spawn Count:NiagaraInt32=$Emitter.SpawnInt, Use Spawn Probability=false)
  ]
  particleSpawnScript [
   &self:DifferentGroup(Max=11, Min=6, RPG:NiagaraInt32=$Emitter.ParticlesPerRibbon, Simple Counter:NiagaraDataInterfaceSimpleCounter=$Emitter.Simple Counter) {module=DifferentGroup}
   &InitializeParticle(
    Color=<rgba 2 2 2 1>
    Lifetime=&FN_SeededRandomRangeFloat(Minimum:NiagaraFloat=0.25, Maximum:NiagaraFloat=0.3, Random Seed:NiagaraInt32=$Particles.Group)
    Ribbon Width=0
    Ribbon Width Mode=::NewEnumerator1
    Sprite Size Mode=::NewEnumerator3
    Uniform Sprite Size=&RandomRangeFloat(Minimum=100, Maximum=150)
    UsePositionOffset=false
    Write Material Random=true
   )
   &CameraOffset(Camera Offset Amount=5)
  ]
  particleUpdateScript [
   &ParticleState()
   &ScaleSpriteSize(
    Scale Sprite Size Mode=::NewEnumerator0
    Uniform Scale Factor=&MakeCustomFloatFromBool(
     Bool=&Boolean_OrOperation(
      Bool1=&SetBoolByFloatComparison(A=$Particles.NormalizeP, B=0, Comparison Type=::NewEnumerator4)
      Bool2=&SetBoolByFloatComparison(A=$Particles.NormalizeP, B=1, Comparison Type=::NewEnumerator4)
     )
     True Float=1
     False Float=0
    )
   )
   &SkeletalMeshLocation(
    Bone / Socket Index=$Particles.Random1
    Bone Sampling Mode=::NewEnumerator4
    MaskSpawnGroup=false
    Orientation Sampling=::NewEnumerator1
    Position Sampling=::NewEnumerator2
    Sampled Position Offset=&RandomRangeVector(Minimum=<-2 -2 -5>, Maximum=<2 2 5>)
    Skeletal Mesh=$User.TargetSM
    Use Sampled Mesh Attribute as Ribbon ID=false
   )
   &SkeletalMeshLocation(
    Bone / Socket Index=$Particles.Random2
    Bone Sampling Mode=::NewEnumerator4
    MaskSpawnGroup=false
    Orientation Sampling=::NewEnumerator1
    Position Sampling=::NewEnumerator2
    Sampled Position Offset=&RandomRangeVector(Minimum=<-2 -2 -5>, Maximum=<2 2 5>)
    Skeletal Mesh=$User.TargetSM
    Use Sampled Mesh Attribute as Ribbon ID=false
   ) {module=SkeletalMeshLocation001}
   /SetVariables(Particles.Position=&LerpPosition(A=$Particles.SkeletalMeshLocation.SampledPosition, B=$Particles.SkeletalMeshLocation001.SampledPosition, Alpha=$Particles.NormalizeP)) {isSetParameters=true, module=SetVariables_#2}
   &self:GetThirdPoint(
    BoneO1:Quat4f=$Particles.SkeletalMeshLocation.SampledRotation
    BoneO2:Quat4f=$Particles.SkeletalMeshLocation001.SampledRotation
    P1:NiagaraPosition=$Output.SkeletalMeshLocation.SampledPositionWithoutOffset
    P2:NiagaraPosition=$Output.SkeletalMeshLocation001.SampledPositionWithoutOffset
    Rotate:NiagaraFloat=&FN_SeededRandomRangeFloat(Minimum:NiagaraFloat=0, Maximum:NiagaraFloat=360, Random Seed:NiagaraInt32=$Particles.Group)
    Se:NiagaraFloat=&MakeCustomFloatFromBool(
     Bool=&SetBoolByFloatComparison(
      A=&FN_SeededRandomRangeFloat(Minimum:NiagaraFloat=0, Maximum:NiagaraFloat=1, Random Seed:NiagaraInt32=$Particles.Group)
      B=0.5
     )
     True Float=0.3
     False Float=-0.3
    )
    float001:NiagaraFloat=&FloatFromCurve(FloatCurve=DataInterface() {propertyValues {Curve {keys [<at 0 1.2>,<at 1 0.8>]}}})
   ) {module=GetThirdPoint}
   &self:BezCurve(
    NormalizedExeindex:NiagaraFloat=$Particles.NormalizeP
    P1:NiagaraPosition=$Particles.SkeletalMeshLocation.SampledPosition
    P2:NiagaraPosition=$Particles.BezCThirdPoint
    P3:NiagaraPosition=$Particles.SkeletalMeshLocation001.SampledPosition
   ) {module=BezCurve}
   &DynamicMaterialParameters(
    Index 0 Param 1=1
    Index 0 Param 3=&FN_SeededRandomRangeFloat(Minimum:NiagaraFloat=75, Maximum:NiagaraFloat=125, Random Seed:NiagaraInt32=$Particles.Group)
    Index 1 Param 3=&FloatFromCurve(FloatCurve=DataInterface() {propertyValues {Curve {keys [<at 0.5 -0.5>,<at 1 0.3>]}}})
    Index 1 Param 4=10
    Index 2 Param 1=&FloatFromCurve(
     FloatCurve=DataInterface() {propertyValues {Curve {keys [
         <at 0 5 RCIM_Cubic>
         {arriveTangent=-0.061, interpMode=RCIM_Cubic, leaveTangent=-0.061, tangentMode=RCTM_User, time=63.8, value=3}
         <at 63.9 0 RCIM_Constant>
        ]}}}
     CurveIndex=&DistanceBetweenPositions(A=$Particles.SkeletalMeshLocation.SampledPosition, B=$Particles.SkeletalMeshLocation001.SampledPosition)
     Scale Curve=2
    )
    Index 3 Param 3=&FloatFromCurve(FloatCurve=DataInterface() {propertyValues {Curve {keys [<at 0 0>,<at 0.8 1>]}}})
    Index 3 Param 4=-1
    Module.Index0Param0Write=true
    Module.Index0Param1Write=true
    Module.Index0Param2Write=true
    Module.Index0Param3Write=true
    Module.Index1Param0Write=true
    Module.Index1Param1Write=true
    Module.Index1Param2Write=true
    Module.Index1Param3Write=true
    Module.Index2Param0Write=true
    Module.Index2Param1Write=true
    Module.Index2Param2Write=true
    Module.Index2Param3Write=true
    Module.Index3Param0Write=true
    Module.Index3Param1Write=true
    Module.Index3Param2Write=true
    Module.Index3Param3Write=true
    Write Parameter Index 1=true
    Write Parameter Index 2=true
    Write Parameter Index 3=true
   )
   &ScaleColor(ScaleA=true, ScaleRGB=true)
   &ScaleColor(
    Scale Alpha=&RandomRangeFloat(Minimum=-0.5, Maximum=1, Evaluation Type=::NewEnumerator1)
    Scale RGB=&VectorFromFloat(Value=&RandomRangeFloat(Minimum=1, Maximum=2, Evaluation Type=::NewEnumerator1))
    ScaleA=true
    ScaleRGB=true
   ) {module=ScaleColor001}
   /SetVariables(Particles.RColor:LinearColor=$Particles.Color) {isSetParameters=true, module=SetVariables_#3}
   &ScaleColor(
    Scale Alpha=&Multiply_Float(
     A=&MakeCustomFloatFromBool(
      Bool=&SetBoolByFloatComparison(A=$Particles.NormalizeP, B=0.5)
      True Float=&FloatFromCurve(FloatCurve=DataInterface() {propertyValues {Curve {keys [<at 0.4 0>,<at 0.6 1>]}}})
      False Float=1
     )
     B=&FloatFromCurve(FloatCurve=DataInterface() {propertyValues {Curve {keys [
          {arriveTangent=7.595, interpMode=RCIM_Cubic, leaveTangent=7.595, tangentMode=RCTM_User, time=0, value=0}
          {interpMode=RCIM_Cubic, tangentMode=RCTM_User, time=0.3, value=1}
          {arriveTangent=0.1069, tangentMode=RCTM_User, time=0.6, value=0}
         ]}}})
    )
    Scale RGB=&MakeVectorFromLinearColorRGB(Color=&MultiplyLinearColor(
      Linear Color=<rgba 0 0.3337 1 1>
      Scale Factor=&MakeVector4_FromFloat(Float=&RandomRangeFloat(Minimum=0, Maximum=2))
     ))
    ScaleA=true
    ScaleRGB=true
   ) {module=ScaleColor002}
  ]
  rendererClasses [&NiagaraSpriteRendererProperties, &NiagaraRibbonRendererProperties]
  renderers [
   {rendererClass=&NiagaraSpriteRendererProperties}
   {rendererClass=&NiagaraRibbonRendererProperties, rendererIndex=1}
  ]
 }
 LeftArm002 {
  emitterUpdateScript [
   &EmitterState(Inactive Response=::NewEnumerator1, Life Cycle Mode=::NewEnumerator1, Loop Behavior=::NewEnumerator1, Loop Duration=4, UseLoopDelay=false)
   /SetVariables() {isSetParameters=true, module=SetVariables_#1}
   &self:SpawnRibbonRate(ParticlesPerRibbon=3, Rate=0.1) {module=SpawnRibbonRate}
   &SpawnPerFrame(Spawn Count:NiagaraInt32=$Emitter.SpawnInt, Use Spawn Probability=false)
  ]
  particleSpawnScript [
   &self:DifferentGroup(Max=15, Min=12, RPG:NiagaraInt32=$Emitter.ParticlesPerRibbon, Simple Counter:NiagaraDataInterfaceSimpleCounter=$Emitter.Simple Counter) {module=DifferentGroup}
   &InitializeParticle(
    Color=<rgba 2 2 2 1>
    Lifetime=&FN_SeededRandomRangeFloat(Minimum:NiagaraFloat=0.25, Maximum:NiagaraFloat=0.3, Random Seed:NiagaraInt32=$Particles.Group)
    Ribbon Width=0
    Ribbon Width Mode=::NewEnumerator1
    Sprite Size Mode=::NewEnumerator3
    Uniform Sprite Size=&RandomRangeFloat(Minimum=100, Maximum=150)
    UsePositionOffset=false
    Write Material Random=true
   )
   &CameraOffset(Camera Offset Amount=5)
  ]
  particleUpdateScript [
   &ParticleState()
   &ScaleSpriteSize(
    Scale Sprite Size Mode=::NewEnumerator0
    Uniform Scale Factor=&MakeCustomFloatFromBool(
     Bool=&Boolean_OrOperation(
      Bool1=&SetBoolByFloatComparison(A=$Particles.NormalizeP, B=0, Comparison Type=::NewEnumerator4)
      Bool2=&SetBoolByFloatComparison(A=$Particles.NormalizeP, B=1, Comparison Type=::NewEnumerator4)
     )
     True Float=1
     False Float=0
    )
   )
   &SkeletalMeshLocation(
    Bone / Socket Index=$Particles.Random1
    Bone Sampling Mode=::NewEnumerator4
    MaskSpawnGroup=false
    Orientation Sampling=::NewEnumerator1
    Position Sampling=::NewEnumerator2
    Sampled Position Offset=&RandomRangeVector(Minimum=<-2 -2 -5>, Maximum=<2 2 5>)
    Skeletal Mesh=$User.TargetSM
    Use Sampled Mesh Attribute as Ribbon ID=false
   )
   &SkeletalMeshLocation(
    Bone / Socket Index=$Particles.Random2
    Bone Sampling Mode=::NewEnumerator4
    MaskSpawnGroup=false
    Orientation Sampling=::NewEnumerator1
    Position Sampling=::NewEnumerator2
    Sampled Position Offset=&RandomRangeVector(Minimum=<-2 -2 -5>, Maximum=<2 2 5>)
    Skeletal Mesh=$User.TargetSM
    Use Sampled Mesh Attribute as Ribbon ID=false
   ) {module=SkeletalMeshLocation001}
   /SetVariables(Particles.Position=&LerpPosition(A=$Particles.SkeletalMeshLocation.SampledPosition, B=$Particles.SkeletalMeshLocation001.SampledPosition, Alpha=$Particles.NormalizeP)) {isSetParameters=true, module=SetVariables_#2}
   &self:GetThirdPoint(
    BoneO1:Quat4f=$Particles.SkeletalMeshLocation.SampledRotation
    BoneO2:Quat4f=$Particles.SkeletalMeshLocation001.SampledRotation
    P1:NiagaraPosition=$Output.SkeletalMeshLocation.SampledPositionWithoutOffset
    P2:NiagaraPosition=$Output.SkeletalMeshLocation001.SampledPositionWithoutOffset
    Rotate:NiagaraFloat=&FN_SeededRandomRangeFloat(Minimum:NiagaraFloat=0, Maximum:NiagaraFloat=360, Random Seed:NiagaraInt32=$Particles.Group)
    Se:NiagaraFloat=&MakeCustomFloatFromBool(
     Bool=&SetBoolByFloatComparison(
      A=&FN_SeededRandomRangeFloat(Minimum:NiagaraFloat=0, Maximum:NiagaraFloat=1, Random Seed:NiagaraInt32=$Particles.Group)
      B=0.5
     )
     True Float=0.3
     False Float=-0.3
    )
    float001:NiagaraFloat=&FloatFromCurve(FloatCurve=DataInterface() {propertyValues {Curve {keys [<at 0 1.2>,<at 1 0.8>]}}})
   ) {module=GetThirdPoint}
   &self:BezCurve(
    NormalizedExeindex:NiagaraFloat=$Particles.NormalizeP
    P1:NiagaraPosition=$Particles.SkeletalMeshLocation.SampledPosition
    P2:NiagaraPosition=$Particles.BezCThirdPoint
    P3:NiagaraPosition=$Particles.SkeletalMeshLocation001.SampledPosition
   ) {module=BezCurve}
   &DynamicMaterialParameters(
    Index 0 Param 1=1
    Index 0 Param 3=&FN_SeededRandomRangeFloat(Minimum:NiagaraFloat=75, Maximum:NiagaraFloat=125, Random Seed:NiagaraInt32=$Particles.Group)
    Index 1 Param 3=&FloatFromCurve(FloatCurve=DataInterface() {propertyValues {Curve {keys [<at 0.5 -0.5>,<at 1 0.3>]}}})
    Index 1 Param 4=10
    Index 2 Param 1=&FloatFromCurve(
     FloatCurve=DataInterface() {propertyValues {Curve {keys [
         <at 0 5 RCIM_Cubic>
         {arriveTangent=-0.061, interpMode=RCIM_Cubic, leaveTangent=-0.061, tangentMode=RCTM_User, time=63.8, value=3}
         <at 63.9 0 RCIM_Constant>
        ]}}}
     CurveIndex=&DistanceBetweenPositions(A=$Particles.SkeletalMeshLocation.SampledPosition, B=$Particles.SkeletalMeshLocation001.SampledPosition)
     Scale Curve=2
    )
    Index 3 Param 3=&FloatFromCurve(FloatCurve=DataInterface() {propertyValues {Curve {keys [<at 0 0>,<at 0.8 1>]}}})
    Index 3 Param 4=-1
    Module.Index0Param0Write=true
    Module.Index0Param1Write=true
    Module.Index0Param2Write=true
    Module.Index0Param3Write=true
    Module.Index1Param0Write=true
    Module.Index1Param1Write=true
    Module.Index1Param2Write=true
    Module.Index1Param3Write=true
    Module.Index2Param0Write=true
    Module.Index2Param1Write=true
    Module.Index2Param2Write=true
    Module.Index2Param3Write=true
    Module.Index3Param0Write=true
    Module.Index3Param1Write=true
    Module.Index3Param2Write=true
    Module.Index3Param3Write=true
    Write Parameter Index 1=true
    Write Parameter Index 2=true
    Write Parameter Index 3=true
   )
   &ScaleColor(ScaleA=true, ScaleRGB=true)
   &ScaleColor(
    Scale Alpha=&RandomRangeFloat(Minimum=-0.5, Maximum=1, Evaluation Type=::NewEnumerator1)
    Scale RGB=&VectorFromFloat(Value=&RandomRangeFloat(Minimum=1, Maximum=2, Evaluation Type=::NewEnumerator1))
    ScaleA=true
    ScaleRGB=true
   ) {module=ScaleColor001}
   /SetVariables(Particles.RColor:LinearColor=$Particles.Color) {isSetParameters=true, module=SetVariables_#3}
   &ScaleColor(
    Scale Alpha=&Multiply_Float(
     A=&MakeCustomFloatFromBool(
      Bool=&SetBoolByFloatComparison(A=$Particles.NormalizeP, B=0.5)
      True Float=&FloatFromCurve(FloatCurve=DataInterface() {propertyValues {Curve {keys [<at 0.4 0>,<at 0.6 1>]}}})
      False Float=1
     )
     B=&FloatFromCurve(FloatCurve=DataInterface() {propertyValues {Curve {keys [
          {arriveTangent=7.595, interpMode=RCIM_Cubic, leaveTangent=7.595, tangentMode=RCTM_User, time=0, value=0}
          {interpMode=RCIM_Cubic, tangentMode=RCTM_User, time=0.3, value=1}
          {arriveTangent=0.1069, tangentMode=RCTM_User, time=0.6, value=0}
         ]}}})
    )
    Scale RGB=&MakeVectorFromLinearColorRGB(Color=&MultiplyLinearColor(
      Linear Color=<rgba 0 0.3337 1 1>
      Scale Factor=&MakeVector4_FromFloat(Float=&RandomRangeFloat(Minimum=0, Maximum=2))
     ))
    ScaleA=true
    ScaleRGB=true
   ) {module=ScaleColor002}
  ]
  rendererClasses [&NiagaraSpriteRendererProperties, &NiagaraRibbonRendererProperties]
  renderers [
   {rendererClass=&NiagaraSpriteRendererProperties}
   {rendererClass=&NiagaraRibbonRendererProperties, rendererIndex=1}
  ]
 }
 LeftArm003 {
  emitterUpdateScript [
   &EmitterState(Inactive Response=::NewEnumerator1, Life Cycle Mode=::NewEnumerator1, Loop Behavior=::NewEnumerator1, Loop Duration=4, UseLoopDelay=false)
   /SetVariables() {isSetParameters=true, module=SetVariables_#1}
   &self:SpawnRibbonRate(ParticlesPerRibbon=3, Rate=0.1) {module=SpawnRibbonRate}
   &SpawnPerFrame(Spawn Count:NiagaraInt32=$Emitter.SpawnInt, Use Spawn Probability=false)
  ]
  particleSpawnScript [
   &self:DifferentGroup(Max=20, Min=16, RPG:NiagaraInt32=$Emitter.ParticlesPerRibbon, Simple Counter:NiagaraDataInterfaceSimpleCounter=$Emitter.Simple Counter) {module=DifferentGroup}
   &InitializeParticle(
    Color=<rgba 2 2 2 1>
    Lifetime=&FN_SeededRandomRangeFloat(Minimum:NiagaraFloat=0.25, Maximum:NiagaraFloat=0.3, Random Seed:NiagaraInt32=$Particles.Group)
    Ribbon Width=0
    Ribbon Width Mode=::NewEnumerator1
    Sprite Size Mode=::NewEnumerator3
    Uniform Sprite Size=&RandomRangeFloat(Minimum=100, Maximum=150)
    UsePositionOffset=false
    Write Material Random=true
   )
   &CameraOffset(Camera Offset Amount=5)
  ]
  particleUpdateScript [
   &ParticleState()
   &ScaleSpriteSize(
    Scale Sprite Size Mode=::NewEnumerator0
    Uniform Scale Factor=&MakeCustomFloatFromBool(
     Bool=&Boolean_OrOperation(
      Bool1=&SetBoolByFloatComparison(A=$Particles.NormalizeP, B=0, Comparison Type=::NewEnumerator4)
      Bool2=&SetBoolByFloatComparison(A=$Particles.NormalizeP, B=1, Comparison Type=::NewEnumerator4)
     )
     True Float=1
     False Float=0
    )
   )
   &SkeletalMeshLocation(
    Bone / Socket Index=$Particles.Random1
    Bone Sampling Mode=::NewEnumerator4
    MaskSpawnGroup=false
    Orientation Sampling=::NewEnumerator1
    Position Sampling=::NewEnumerator2
    Sampled Position Offset=&RandomRangeVector(Minimum=<-2 -2 -5>, Maximum=<2 2 5>)
    Skeletal Mesh=$User.TargetSM
    Use Sampled Mesh Attribute as Ribbon ID=false
   )
   &SkeletalMeshLocation(
    Bone / Socket Index=$Particles.Random2
    Bone Sampling Mode=::NewEnumerator4
    MaskSpawnGroup=false
    Orientation Sampling=::NewEnumerator1
    Position Sampling=::NewEnumerator2
    Sampled Position Offset=&RandomRangeVector(Minimum=<-2 -2 -5>, Maximum=<2 2 5>)
    Skeletal Mesh=$User.TargetSM
    Use Sampled Mesh Attribute as Ribbon ID=false
   ) {module=SkeletalMeshLocation001}
   /SetVariables(Particles.Position=&LerpPosition(A=$Particles.SkeletalMeshLocation.SampledPosition, B=$Particles.SkeletalMeshLocation001.SampledPosition, Alpha=$Particles.NormalizeP)) {isSetParameters=true, module=SetVariables_#2}
   &self:GetThirdPoint(
    BoneO1:Quat4f=$Particles.SkeletalMeshLocation.SampledRotation
    BoneO2:Quat4f=$Particles.SkeletalMeshLocation001.SampledRotation
    P1:NiagaraPosition=$Output.SkeletalMeshLocation.SampledPositionWithoutOffset
    P2:NiagaraPosition=$Output.SkeletalMeshLocation001.SampledPositionWithoutOffset
    Rotate:NiagaraFloat=&FN_SeededRandomRangeFloat(Minimum:NiagaraFloat=0, Maximum:NiagaraFloat=360, Random Seed:NiagaraInt32=$Particles.Group)
    Se:NiagaraFloat=&MakeCustomFloatFromBool(
     Bool=&SetBoolByFloatComparison(
      A=&FN_SeededRandomRangeFloat(Minimum:NiagaraFloat=0, Maximum:NiagaraFloat=1, Random Seed:NiagaraInt32=$Particles.Group)
      B=0.5
     )
     True Float=0.3
     False Float=-0.3
    )
    float001:NiagaraFloat=&FloatFromCurve(FloatCurve=DataInterface() {propertyValues {Curve {keys [<at 0 1.2>,<at 1 0.8>]}}})
   ) {module=GetThirdPoint}
   &self:BezCurve(
    NormalizedExeindex:NiagaraFloat=$Particles.NormalizeP
    P1:NiagaraPosition=$Particles.SkeletalMeshLocation.SampledPosition
    P2:NiagaraPosition=$Particles.BezCThirdPoint
    P3:NiagaraPosition=$Particles.SkeletalMeshLocation001.SampledPosition
   ) {module=BezCurve}
   &DynamicMaterialParameters(
    Index 0 Param 1=1
    Index 0 Param 3=&FN_SeededRandomRangeFloat(Minimum:NiagaraFloat=75, Maximum:NiagaraFloat=125, Random Seed:NiagaraInt32=$Particles.Group)
    Index 1 Param 3=&FloatFromCurve(FloatCurve=DataInterface() {propertyValues {Curve {keys [<at 0.5 -0.5>,<at 1 0.3>]}}})
    Index 1 Param 4=10
    Index 2 Param 1=&FloatFromCurve(
     FloatCurve=DataInterface() {propertyValues {Curve {keys [
         <at 0 5 RCIM_Cubic>
         {arriveTangent=-0.061, interpMode=RCIM_Cubic, leaveTangent=-0.061, tangentMode=RCTM_User, time=63.8, value=3}
         <at 63.9 0 RCIM_Constant>
        ]}}}
     CurveIndex=&DistanceBetweenPositions(A=$Particles.SkeletalMeshLocation.SampledPosition, B=$Particles.SkeletalMeshLocation001.SampledPosition)
     Scale Curve=2
    )
    Index 3 Param 3=&FloatFromCurve(FloatCurve=DataInterface() {propertyValues {Curve {keys [<at 0 0>,<at 0.8 1>]}}})
    Index 3 Param 4=-1
    Module.Index0Param0Write=true
    Module.Index0Param1Write=true
    Module.Index0Param2Write=true
    Module.Index0Param3Write=true
    Module.Index1Param0Write=true
    Module.Index1Param1Write=true
    Module.Index1Param2Write=true
    Module.Index1Param3Write=true
    Module.Index2Param0Write=true
    Module.Index2Param1Write=true
    Module.Index2Param2Write=true
    Module.Index2Param3Write=true
    Module.Index3Param0Write=true
    Module.Index3Param1Write=true
    Module.Index3Param2Write=true
    Module.Index3Param3Write=true
    Write Parameter Index 1=true
    Write Parameter Index 2=true
    Write Parameter Index 3=true
   )
   &ScaleColor(ScaleA=true, ScaleRGB=true)
   &ScaleColor(
    Scale Alpha=&RandomRangeFloat(Minimum=-0.5, Maximum=1, Evaluation Type=::NewEnumerator1)
    Scale RGB=&VectorFromFloat(Value=&RandomRangeFloat(Minimum=1, Maximum=2, Evaluation Type=::NewEnumerator1))
    ScaleA=true
    ScaleRGB=true
   ) {module=ScaleColor001}
   /SetVariables(Particles.RColor:LinearColor=$Particles.Color) {isSetParameters=true, module=SetVariables_#3}
   &ScaleColor(
    Scale Alpha=&Multiply_Float(
     A=&MakeCustomFloatFromBool(
      Bool=&SetBoolByFloatComparison(A=$Particles.NormalizeP, B=0.5)
      True Float=&FloatFromCurve(FloatCurve=DataInterface() {propertyValues {Curve {keys [<at 0.4 0>,<at 0.6 1>]}}})
      False Float=1
     )
     B=&FloatFromCurve(FloatCurve=DataInterface() {propertyValues {Curve {keys [
          {arriveTangent=7.595, interpMode=RCIM_Cubic, leaveTangent=7.595, tangentMode=RCTM_User, time=0, value=0}
          {interpMode=RCIM_Cubic, tangentMode=RCTM_User, time=0.3, value=1}
          {arriveTangent=0.1069, tangentMode=RCTM_User, time=0.6, value=0}
         ]}}})
    )
    Scale RGB=&MakeVectorFromLinearColorRGB(Color=&MultiplyLinearColor(
      Linear Color=<rgba 0 0.3337 1 1>
      Scale Factor=&MakeVector4_FromFloat(Float=&RandomRangeFloat(Minimum=0, Maximum=2))
     ))
    ScaleA=true
    ScaleRGB=true
   ) {module=ScaleColor002}
  ]
  rendererClasses [&NiagaraSpriteRendererProperties, &NiagaraRibbonRendererProperties]
  renderers [
   {rendererClass=&NiagaraSpriteRendererProperties}
   {rendererClass=&NiagaraRibbonRendererProperties, rendererIndex=1}
  ]
 }
}
summary {
 systemName=NS_ZeusGodAI_LightningAmbientFX
 emitters [
  {emitterName=LeftArm, rendererClasses [&NiagaraSpriteRendererProperties, &NiagaraRibbonRendererProperties]}
  {emitterName=LeftArm001, rendererClasses [&NiagaraSpriteRendererProperties, &NiagaraRibbonRendererProperties]}
  {emitterName=LeftArm002, rendererClasses [&NiagaraSpriteRendererProperties, &NiagaraRibbonRendererProperties]}
  {emitterName=LeftArm003, rendererClasses [&NiagaraSpriteRendererProperties, &NiagaraRibbonRendererProperties]}
  {emitterName=Empty, rendererClasses [&NiagaraMeshRendererProperties]}
  {emitterName=Empty001, rendererClasses [&NiagaraSpriteRendererProperties, &NiagaraLightRendererProperties]}
 ]
 userVariables [User.TargetSM {
   type=NiagaraDataInterfaceSkeletalMesh
   defaultValue {
    struct=NiagaraExt_VariableValue_DataInterface
    value {dataInterface=&self:NiagaraDataInterfaceSkeletalMesh_0}
   }
  }]
}
